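# Niagara glider contrail for the SamuraiBlueTrail cosmetic, drawing two ribbon trails from bone or socket pairs on a skeletal mesh.
# Shared AnimTrail modules spawn and update the ribbons, with added velocity for front-end preview.
# Alpha combines a lifetime curve, delayed emitter-age fade-in, time-of-day dimming and, outside preview, a system-smoothed interpolation value.
# The speed-line sprite emitter and the Trail's blue-tinted color module are disabled.
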NiagaraSystem /BRCosmetics/Athena/Effects/Particles/Gliders/NS_GliderContrail_SamuraiBlueTrail
dependencies {
 usedDataInterfaces [
  &NiagaraDataInterfaceColorCurve
  &NiagaraDataInterfaceCurve
  &NiagaraDataInterfaceVector2DCurve
  &NiagaraDataInterfaceSkeletalMesh
 ]
 usedDynamicInputs []
 usedModules [
  &self:SystemSpawnScript.SystemScriptSource.SystemScriptGraph.NiagaraNodeAssignment_0.SetVariables_#1
  &SystemState
  &InterpolateOverTime
 ]
}
emitters {
 SppedLine {
  bEnabled=false
  emitterUpdateScript [
   &EmitterState(
    Delay First Loop Only=true
    Inactive Response=::NewEnumerator1
    Life Cycle Mode=::NewEnumerator1
    Loop Delay=1
    Loop Duration=1
    UseLoopDelay=true
   )
   &SpawnRate(
    SpawnRate=&MakeCustomFloatFromBool(Bool=$User.bIsFrontEndPreview, True Float=150, False Float=150)
    Use Spawn Probability=true
   ) {module=SpawnRate001}
  ]
  particleSpawnScript [
   &InitializeParticle(
    Color=&MakeLinearColorFromVectorAndFloat(
     "Vector (RGB)"=<1 1 1>
     "Float (Alpha)"=&Multiply_Float(
      A=&UniformRangedFloat(Minimum=0.1, Maximum=0.2)
      B=&MakeCustomFloatFromBool(Bool=$User.bIsFrontEndPreview, True Float=0.5, False Float=1)
     )
    )
    Lifetime Max=0.1
    Lifetime Min=0.1
    Lifetime Mode=::NewEnumerator1
    Sprite Size Max=<2 10>
    Sprite Size Min=<2 2>
    Sprite Size Mode=::NewEnumerator2
    UsePositionOffset=false
    Write Material Random=true
   )
   &ShapeLocation(
    Handle Radius=0
    Large Radius=&RandomRangeFloat(Minimum=100, Maximum=150)
    MaskSpawnGroup=false
    Override Seed=false
    Rotation Axis=<0 0 1>
    Rotation Mode=::NewEnumerator1
    Shape Origin=$Engine.Emitter.SimulationPosition
    Shape Primitive=::NewEnumerator3
    UseOwnerScale=true
   )
   &AddVelocity(Module.Use Falloff Distance=false, Override Seed=false, Velocity Mode=::NewEnumerator1, Velocity Speed=100) {module=AddVelocity001}
   &AddVelocity(Velocity=&Multiply_Vector(
     A=&RandomRangeVector(Minimum=<0 0 50>, Maximum=<0 0 200>)
     B=&MakeVector(X=1, Y=&MakeCustomFloatFromBool(Bool=$User.bIsFrontEndPreview, True Float=-1, False Float=1), Z=1)
    ))
  ]
  particleUpdateScript [
   &ParticleState()
   &ScaleSpriteSize(Scale Factor=&Vector2DFromCurve(
     Vector2Curve=DataInterface() {propertyValues {XCurve {keys [<at 0 0.5 RCIM_Cubic>,<at 0.5 1 RCIM_Cubic>]}, YCurve {keys [<at 0 0 RCIM_Cubic>,<at 0.5 1 RCIM_Cubic>]}}}
     ScaleCurve=&Vector2DFromFloat(Value=1)
    ))
   &SolveForcesAndVelocity(Clamp Velocity=false, Limit Acceleration=false)
   &ScaleSpriteSizeBySpeed(Max Scale Factor=<1 30>, Min Scale Factor=<0.01 0.05>, Velocity Threshold=1)
   &ScaleColor(
    Scale Alpha=&FloatFromCurve(
     FloatCurve=DataInterface() {propertyValues {Curve {keys [
         <at 0 0 RCIM_Cubic>
         <at 0.37 1.0207>
         {arriveTangent=-2.221, interpMode=RCIM_Cubic, leaveTangent=-2.221, tangentMode=RCTM_User, time=1, value=0}
        ]}}}
     Scale Curve=&MakeCustomFloatFromBool(
      Bool=$User.bIsFrontEndPreview
      True Float=&MakeCustomFloatFromBool(Bool=$User.bIsFullyDeployed, True Float=1, False Float=0)
      False Float=&Lerp_Float(A=0, B=1, Alpha=$System.InterpolateOverTime.InterpolateOverTime_Float.Moving Average)
     )
    )
    Scale RGB=&VectorFromFloat(Value=1)
    ScaleA=true
    ScaleRGB=true
   )
  ]
  rendererClasses [&NiagaraSpriteRendererProperties]
  renderers [{rendererClass=&NiagaraSpriteRendererProperties}]
 }
 Trail {
  emitterUpdateScript [
   &EmitterState(
    Inactive Response=::NewEnumerator1
    Life Cycle Mode=::NewEnumerator1
    Loop Behavior=::NewEnumerator1
    Loop Duration Mode=::NewEnumerator1
    UseLoopDelay=false
   )
   &FN_SpawnAnimTrail(Number of Anim Trails=2)
  ]
  particleSpawnScript [
   &FN_InitializeAnimTrailBoneIndecies()
   &FN_InitializeAnimTrailBoneIndecies(AnimTrail Selection=1, First Bone or Socket Index=2, Second Bone or Socket Index=3) {module=FN_InitializeAnimTrailBoneIndecies001}
   &FN_InitializeAnimTrail(AnimTrailWidth=1.2, Color=<rgba 5 5 5 1>, Lifetime=1, SkeletalMesh=$System.Skeletal Mesh Trail)
   &CameraOffset(Camera Offset Amount=-20) {enabled=false}
   &FN_Sprite_Attributes(
    Color=&MakeLinearColorFromVectorAndFloat(
     "Vector (RGB)"=&Multiply_Vector(
      A=&MakeVectorFromLinearColorRGB(Color=<rgba 0.9297 1.2047 1.5 0.2>)
      B=&MakeVectorFromLinearColorRGB(Color=&ColorFromCurve(ColorCurve=DataInterface() {propertyValues {
          AlphaCurve {keys []}
          BlueCurve {keys [<at 0 1 RCIM_Cubic>,<at 0.2993 0.5 RCIM_Cubic>]}
          GreenCurve {keys [<at 0 0.7645 RCIM_Cubic>,<at 0.2993 0.5 RCIM_Cubic>]}
          RedCurve {keys [<at 0 0.4077 RCIM_Cubic>,<at 0.2993 0.5 RCIM_Cubic>]}
         }}))
     )
     "Float (Alpha)"=&Multiply_Float(
      A=1
      B=&FloatFromCurve(
       FloatCurve=DataInterface() {propertyValues {Curve {keys [
           {interpMode=RCIM_Cubic, tangentMode=RCTM_User, time=0, value=0}
           <at 0.2 0.9 RCIM_Cubic>
           {interpMode=RCIM_Cubic, tangentMode=RCTM_User, time=1, value=0}
          ]}}}
       Scale Curve=&Multiply_Float(
        A=&Multiply_Float(
         A=2
         B=&FloatFromCurve(FloatCurve=DataInterface() {propertyValues {Curve {keys [<at 1 0>,<at 2 1>]}}}, CurveIndex=$Emitter.Age)
        )
        B=$System.InterpolateOverTime.InterpolateOverTime_Float.Moving Average
       )
      )
     )
    )
    Create Color Attribute=true
    Write to Module Namespace=true
   ) {enabled=false}
   &AddVelocity(Velocity=&FN_FrontEndVector3Modifier(Gameplay=<0 0 0>, Front End=<600 0 0>))
  ]
  particleUpdateScript [
   /SetVariables(Particles.Linear Color02:LinearColor=&MakeLinearColorFromVectorAndFloat(
     "Vector (RGB)"=<1 1 1>
     "Float (Alpha)"=&Multiply_Float(
      A=0.2
      B=&FloatFromCurve(
       FloatCurve=DataInterface() {propertyValues {Curve {keys [
           {interpMode=RCIM_Cubic, tangentMode=RCTM_User, time=0, value=0}
           <at 0.2 0.9 RCIM_Cubic>
           {interpMode=RCIM_Cubic, tangentMode=RCTM_User, time=1, value=0}
          ]}}}
       Scale Curve=&Multiply_Float(
        A=&Multiply_Float(
         A=2
         B=&FloatFromCurve(FloatCurve=DataInterface() {propertyValues {Curve {keys [<at 1 0>,<at 2 1>]}}}, CurveIndex=$Emitter.Age)
        )
        B=$System.InterpolateOverTime.InterpolateOverTime_Float.Moving Average
       )
      )
     )
    )) {isSetParameters=true}
   &ParticleState()
   &FN_UpdateAnimTrail(SkeletalMesh=$System.Skeletal Mesh Trail)
   &ScaleColor(
    Scale Alpha=&FloatFromCurve(
     FloatCurve=DataInterface() {propertyValues {Curve {keys [
         {arriveTangent=-0.0593, interpMode=RCIM_Cubic, leaveTangent=-0.0593, tangentMode=RCTM_User, time=0, value=0}
         {arriveTangent=-0.837, interpMode=RCIM_Cubic, leaveTangent=-0.837, tangentMode=RCTM_User, time=0.05, value=1}
         {arriveTangent=-0.0593, interpMode=RCIM_Cubic, leaveTangent=-0.0593, tangentMode=RCTM_User, time=1, value=0}
        ]}}}
     Scale Curve=&Multiply_Float(
      A=&Multiply_Float(
       A=&FNTimeOfDayScalar(Night Scale=0.4, Morning Scale=1, Evening Scale=1)
       B=&FloatFromCurve(
        FloatCurve=DataInterface() {propertyValues {Curve {keys [<at 1 0>,<at 2 1>]}}}
        CurveIndex=$Emitter.Age
        Scale Curve=&MakeCustomFloatFromBool(Bool=$User.bIsFrontEndPreview, True Float=0.25, False Float=1)
       )
      )
      B=&MakeCustomFloatFromBool(Bool=$User.bIsFrontEndPreview, True Float=1, False Float=$System.InterpolateOverTime.InterpolateOverTime_Float.Moving Average)
     )
    )
    Scale RGB=&MakeVectorFromLinearColorRGB(Color=<rgba 1 1 1 1>)
    ScaleA=true
    ScaleRGB=true
   )
   &SolveForcesAndVelocity(Clamp Velocity=false, Limit Acceleration=false)
  ]
  rendererClasses [&NiagaraRibbonRendererProperties]
  renderers [{rendererClass=&NiagaraRibbonRendererProperties}]
 }
}
summary {
 systemName=NS_GliderContrail_SamuraiBlueTrail
 emitters [
  {emitterName=Trail, rendererClasses [&NiagaraRibbonRendererProperties]}
  {bEnabled=false, emitterName=SppedLine, rendererClasses [&NiagaraSpriteRendererProperties]}
 ]
 userVariables [
  User.bIsFrontEnd {type=NiagaraBool}
  User.bIsFrontEndPreview {type=NiagaraBool}
  User.bIsFullyDeployed {type=NiagaraBool}
  User.ForwardDot {type=NiagaraFloat, defaultValue=0, description=""}
  User.RibbonIDOffset {type=NiagaraInt32}
  User.RightDot {type=NiagaraFloat, defaultValue=0, description=""}
 ]
}
