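# Niagara glider contrail for StreetRacerBlack, converted per its name, that draws two socket- or bone-driven ribbon trails from a single emitter, Trail_L.
# FN_SpawnAnimTrail spawns both trails, two index initializers assign each its bone or socket pair, and FN_InitializeAnimTrail sets a bright red color and short fixed lifetime.
# FN_UpdateAnimTrail updates the trails while an alpha curve fades them in and out.
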NiagaraSystem /BRCosmetics/Athena/Effects/Particles/Gliders/NS_GliderContrail_StreetRacerBlack_Converted
dependencies {
 usedDataInterfaces [&NiagaraDataInterfaceVectorCurve, &NiagaraDataInterfaceCurve, &NiagaraDataInterfaceSocketReader]
 usedDynamicInputs []
 usedModules [
  &self:SystemSpawnScript.SystemScriptSource.SystemScriptGraph.NiagaraNodeAssignment_0.SetVariables_#1
  &SystemState
 ]
}
emitters {Trail_L {
  emitterUpdateScript [
   &EmitterState(Life Cycle Mode=::NewEnumerator1, Loop Duration=1, UseLoopDelay=true)
   &FN_SpawnAnimTrail(Number of Anim Trails=2, Spawn Interval=0)
  ]
  particleSpawnScript [
   &Color(Color=&MakeLinearColorFromVectorAndFloat("Vector (RGB)"=<30 30 30>, "Float (Alpha)"=1))
   &FN_InitializeAnimTrailBoneIndecies()
   &FN_InitializeAnimTrailBoneIndecies(AnimTrail Selection=1, First Bone or Socket Index=2, Second Bone or Socket Index=3) {module=FN_InitializeAnimTrailBoneIndecies001}
   &FN_InitializeAnimTrail(Color=<rgba 30 0.99 0.99 1>, Lifetime=0.475)
  ]
  particleUpdateScript [
   &ParticleState()
   &ScaleColor(
    Scale Alpha=&FloatFromCurve(FloatCurve=DataInterface() {propertyValues {Curve {keys [<at 0 0>,<at 0.2 1>,<at 0.7 1>,<at 1 0>]}}})
    Scale RGB=&VectorFromCurve(VectorCurve=DataInterface() {propertyValues {XCurve {keys [<at 0 1>,<at 1 1>]}, YCurve {keys [<at 0 1>,<at 1 1>]}, ZCurve {keys [<at 0 1>,<at 1 1>]}}})
    ScaleA=true
    ScaleRGB=true
   )
   &FN_UpdateAnimTrail()
  ]
  rendererClasses [&NiagaraRibbonRendererProperties]
  renderers [{rendererClass=&NiagaraRibbonRendererProperties}]
 }}
summary {
 systemName=NS_GliderContrail_StreetRacerBlack_Converted
 emitters [{emitterName=Trail_L, rendererClasses [&NiagaraRibbonRendererProperties]}]
 userVariables [
  User.RibbonIDOffset {type=NiagaraInt32}
  User.Trail_Life {type=NiagaraFloat, defaultValue=0, description=""}
 ]
}
