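# Niagara effect for the Street Racer Black glider contrail, a low light variant that by its naming renders a glowing red tail light.
# A single TailLight emitter continuously spawns a few stretched mesh particles per second with an HDR red color and no motion modules.
# Over each particle's life, curves quickly scale the mesh up from zero and fade alpha in and back out, so overlapping particles form the light.
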NiagaraSystem /BRCosmetics/Athena/Effects/Particles/Gliders/NS_GliderContrail_StreetRacerBlack_LightLow_Converted
dependencies {usedDataInterfaces [&NiagaraDataInterfaceCurve], usedDynamicInputs [], usedModules [&SystemState]}
emitters {TailLight {
  emitterUpdateScript [
   &EmitterState(Inactive Response=::NewEnumerator1, Life Cycle Mode=::NewEnumerator1, Loop Duration=1, UseLoopDelay=true)
   &SpawnRate(SpawnRate=3, Use Spawn Probability=false)
  ]
  particleSpawnScript [
   &InitializeParticle(Mesh Scale=<0.08 0.8 1>, Mesh Scale Mode=::NewEnumerator4, UsePositionOffset=false, Write Material Random=true)
   &Color(Color=&MakeLinearColorFromVectorAndFloat("Vector (RGB)"=<30 0.0666 0>, "Float (Alpha)"=1))
  ]
  particleUpdateScript [
   &ParticleState()
   &ScaleMeshSize(Scale Factor=&VectorFromFloat(Value=&FloatFromCurve(FloatCurve=DataInterface() {propertyValues {Curve {keys [<at 0 0>,<at 0.05 1>]}}})))
   &ScaleColor(
    Scale Alpha=&FloatFromCurve(FloatCurve=DataInterface() {propertyValues {Curve {keys [
         <at 0 0 RCIM_Cubic>
         {arriveTangent=1.6396, interpMode=RCIM_Cubic, leaveTangent=1.6396, time=0.0434, value=0.9756}
         {arriveTangent=-1.0198, interpMode=RCIM_Cubic, leaveTangent=-1.0198, time=0.6102, value=1.0006}
         <at 1 0>
        ]}}})
    ScaleA=true
    ScaleRGB=true
   )
  ]
  rendererClasses [&NiagaraMeshRendererProperties]
  renderers [{rendererClass=&NiagaraMeshRendererProperties, rendererIndex=0}]
 }}
summary {
 systemName=NS_GliderContrail_StreetRacerBlack_LightLow_Converted
 emitters [{emitterName=TailLight, rendererClasses [&NiagaraMeshRendererProperties]}]
 userVariables []
}
