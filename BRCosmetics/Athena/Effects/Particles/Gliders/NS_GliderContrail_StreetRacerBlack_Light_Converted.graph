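# Niagara glider contrail effect for the StreetRacerBlack glider that adds glowing red lights at the left and right trail sockets, using light renderers only.
# Mirrored LightL and LightR emitters continuously spawn bright red particles, with a curve fading their alpha in and out over each lifetime.
# Skeletal mesh sampling in particle update applies FX_Trail_L/R socket locations with opposite sideways offsets, and a Cascade-conversion module sets light brightness and radius.
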NiagaraSystem /BRCosmetics/Athena/Effects/Particles/Gliders/NS_GliderContrail_StreetRacerBlack_Light_Converted
dependencies {
 usedDataInterfaces [&NiagaraDataInterfaceCurve, &NiagaraDataInterfaceSkeletalMesh]
 usedDynamicInputs []
 usedModules [&SystemState]
}
emitters {
 LightL {
  emitterUpdateScript [
   &EmitterState(Inactive Response=::NewEnumerator1, Life Cycle Mode=::NewEnumerator1, Loop Duration=1, UseLoopDelay=true)
   &SpawnRate(SpawnRate=3, Use Spawn Probability=false)
  ]
  particleSpawnScript [
   &InitializeParticle(Sprite Size=<50 50>, Sprite Size Mode=::NewEnumerator4, UsePositionOffset=false, Write Material Random=true)
   &Color(Color=&MakeLinearColorFromVectorAndFloat("Vector (RGB)"=<30 0.05 0>, "Float (Alpha)"=1.5))
  ]
  particleUpdateScript [
   /SetVariables(Particles.LightExponent=8) {isSetParameters=true}
   &ParticleState()
   &ScaleColor(
    Scale Alpha=&FloatFromCurve(FloatCurve=DataInterface() {propertyValues {Curve {keys [
         <at 0 0 RCIM_Cubic>
         {arriveTangent=1.6396, interpMode=RCIM_Cubic, leaveTangent=1.6396, time=0.0434, value=0.9756}
         {arriveTangent=-1.0198, interpMode=RCIM_Cubic, leaveTangent=-1.0198, time=0.6102, value=1.0006}
         <at 1 0>
        ]}}})
    ScaleA=true
    ScaleRGB=true
   )
   &CascadeConversion_LightProperties(LightBrightnessScale=16, LightColorScale=<1 1 1>, LightRadiusScale=1.25)
   &SkeletalMeshLocation(
    MaskSpawnGroup=false
    Mesh Sampling Type=::NewEnumerator5
    Orientation Sampling=::NewEnumerator1
    Sampled Position Offset=<8 50 0>
    Sampled Position Offset Coordinate Space:ENiagaraSkelMeshTransforms=::NewEnumerator2
    Skeletal Mesh=DataInterface() {propertyValues {
      DefaultMesh=null
      ExcludeBoneName=root
      OptionalComponentClass=null
      PreviewMesh=/BRCosmetics/Gadgets/Assets/VinderTech_GliderChute/Meshes/SK_M_VinderTech_GliderChute
      SkinningMode=SkinOnTheFly
      SoftSourceActor=null
      SourceMode=Default
      UvSetIndex=0
      WholeMeshLOD=-1
      bExcludeBone=false
      bReadDeformedGeometry=true
      bRequireCurrentFrameData=true
      ComponentTags []
      FilteredBones []
      FilteredSockets [FX_Trail_L]
      MeshUserParameter {parameter {name=None, typeDefHandle {registeredTypeIndex=1}, varData []}}
      SamplingRegions []
     }}
    Socket Sampling Mode=::NewEnumerator5
   )
  ]
  rendererClasses [&NiagaraLightRendererProperties]
  renderers [{rendererClass=&NiagaraLightRendererProperties, rendererIndex=0}]
 }
 LightR {
  emitterUpdateScript [
   &EmitterState(Inactive Response=::NewEnumerator1, Life Cycle Mode=::NewEnumerator1, Loop Duration=1, UseLoopDelay=true)
   &SpawnRate(SpawnRate=3, Use Spawn Probability=false)
  ]
  particleSpawnScript [
   &InitializeParticle(Sprite Size=<50 50>, Sprite Size Mode=::NewEnumerator4, UsePositionOffset=false, Write Material Random=true)
   &Color(Color=&MakeLinearColorFromVectorAndFloat("Vector (RGB)"=<30 0.05 0>, "Float (Alpha)"=1.5))
  ]
  particleUpdateScript [
   /SetVariables(Particles.LightExponent=8) {isSetParameters=true}
   &ParticleState()
   &ScaleColor(
    Scale Alpha=&FloatFromCurve(FloatCurve=DataInterface() {propertyValues {Curve {keys [
         <at 0 0 RCIM_Cubic>
         {arriveTangent=1.6396, interpMode=RCIM_Cubic, leaveTangent=1.6396, time=0.0434, value=0.9756}
         {arriveTangent=-1.0198, interpMode=RCIM_Cubic, leaveTangent=-1.0198, time=0.6102, value=1.0006}
         <at 1 0>
        ]}}})
    ScaleA=true
    ScaleRGB=true
   )
   &CascadeConversion_LightProperties(LightBrightnessScale=16, LightColorScale=<1 1 1>, LightRadiusScale=1.25)
   &SkeletalMeshLocation(
    MaskSpawnGroup=false
    Mesh Sampling Type=::NewEnumerator5
    Orientation Sampling=::NewEnumerator1
    Sampled Position Offset=<-8 50 0>
    Sampled Position Offset Coordinate Space:ENiagaraSkelMeshTransforms=::NewEnumerator2
    Skeletal Mesh=DataInterface() {propertyValues {
      DefaultMesh=null
      ExcludeBoneName=root
      OptionalComponentClass=null
      PreviewMesh=&SK_M_VinderTech_GliderChute
      SkinningMode=SkinOnTheFly
      SoftSourceActor=null
      SourceMode=Default
      UvSetIndex=0
      WholeMeshLOD=-1
      bExcludeBone=false
      bReadDeformedGeometry=true
      bRequireCurrentFrameData=true
      ComponentTags []
      FilteredBones []
      FilteredSockets [FX_Trail_R]
      MeshUserParameter {parameter {name=None, typeDefHandle {registeredTypeIndex=1}, varData []}}
      SamplingRegions []
     }}
    Socket Sampling Mode=::NewEnumerator5
   )
  ]
  rendererClasses [&NiagaraLightRendererProperties]
  renderers [{rendererClass=&NiagaraLightRendererProperties, rendererIndex=0}]
 }
}
summary {
 systemName=NS_GliderContrail_StreetRacerBlack_Light_Converted
 emitters [
  {emitterName=LightL, rendererClasses [&NiagaraLightRendererProperties]}
  {emitterName=LightR, rendererClasses [&NiagaraLightRendererProperties]}
 ]
 userVariables []
}
