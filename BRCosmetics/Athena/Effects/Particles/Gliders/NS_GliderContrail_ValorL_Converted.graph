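# Niagara system for a cosmetic glider contrail (ValorL) that draws two white ribbon trails from the glider.
# Both emitters use anim-trail spawn, initialize and update modules, each spanning its own pair of bone or socket indices.
# A step curve on system age keeps newly spawned particles transparent for roughly the first 0.7 seconds.
# Over each short particle lifetime, trails start overbright and dim while alpha fades in, then out.
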NiagaraSystem /BRCosmetics/Athena/Effects/Particles/Gliders/NS_GliderContrail_ValorL_Converted
dependencies {
 usedDataInterfaces [&NiagaraDataInterfaceCurve, &NiagaraDataInterfaceVectorCurve, &NiagaraDataInterfaceSocketReader]
 usedDynamicInputs []
 usedModules [
  &self:SystemSpawnScript.SystemScriptSource.SystemScriptGraph.NiagaraNodeAssignment_0.SetVariables_#1
  &SystemState
 ]
}
emitters {
 Trail_R {
  emitterUpdateScript [
   &EmitterState()
   &FN_SpawnAnimTrail(AnimTrailWidthMode=::FromFirst)
  ]
  particleSpawnScript [
   &FN_InitializeAnimTrailBoneIndecies(First Bone or Socket Index=1, Second Bone or Socket Index=0)
   &FN_InitializeAnimTrail(
    AnimTrailWidth=0.8
    Color=&MakeLinearColorFromVectorAndFloat(
     "Vector (RGB)"=<1 1 1>
     "Float (Alpha)"=&FloatFromCurve(FloatCurve=DataInterface() {propertyValues {Curve {keys [<at 0.699 0>,<at 0.7 1>]}}}, CurveIndex=$System.Age)
    )
    Lifetime=0.4
   )
  ]
  particleUpdateScript [
   &ParticleState()
   &FN_UpdateAnimTrail()
   &ScaleColor(
    Scale Alpha=&FloatFromCurve(FloatCurve=DataInterface() {propertyValues {Curve {keys [<at 0.0008 0 RCIM_Cubic>,<at 0.2357 1 RCIM_Cubic>,<at 1 0 RCIM_Cubic>]}}})
    Scale RGB=&VectorFromCurve(VectorCurve=DataInterface() {propertyValues {
       XCurve {keys [<at 0 3 RCIM_Cubic>, {arriveTangent=-8.3333, interpMode=RCIM_Cubic, leaveTangent=-8.3333, time=0.15, value=1}, <at 0.3 0.5 RCIM_Cubic>]}
       YCurve {keys [<at 0 3 RCIM_Cubic>, {arriveTangent=-8.3333, interpMode=RCIM_Cubic, leaveTangent=-8.3333, time=0.15, value=1}, <at 0.3 0.5 RCIM_Cubic>]}
       ZCurve {keys [<at 0 3 RCIM_Cubic>, {arriveTangent=-8.3333, interpMode=RCIM_Cubic, leaveTangent=-8.3333, time=0.15, value=1}, <at 0.3 0.5 RCIM_Cubic>]}
      }})
    ScaleA=true
    ScaleRGB=true
   )
  ]
  rendererClasses [&NiagaraRibbonRendererProperties]
  renderers [{rendererClass=&NiagaraRibbonRendererProperties}]
 }
 Trail_R001 {
  emitterUpdateScript [
   &EmitterState()
   &FN_SpawnAnimTrail(AnimTrailWidthMode=::FromFirst)
  ]
  particleSpawnScript [
   &FN_InitializeAnimTrailBoneIndecies(First Bone or Socket Index=3, Second Bone or Socket Index=2) {module=FN_InitializeAnimTrailBoneIndecies001}
   &FN_InitializeAnimTrail(
    AnimTrailWidth=0.8
    Color=&MakeLinearColorFromVectorAndFloat(
     "Vector (RGB)"=<1 1 1>
     "Float (Alpha)"=&FloatFromCurve(FloatCurve=DataInterface() {propertyValues {Curve {keys [<at 0.699 0>,<at 0.7 1>]}}}, CurveIndex=$System.Age)
    )
    Lifetime=0.4
   )
  ]
  particleUpdateScript [
   &ParticleState()
   &FN_UpdateAnimTrail()
   &ScaleColor(
    Scale Alpha=&FloatFromCurve(FloatCurve=DataInterface() {propertyValues {Curve {keys [<at 0.0008 0 RCIM_Cubic>,<at 0.2357 1 RCIM_Cubic>,<at 1 0 RCIM_Cubic>]}}})
    Scale RGB=&VectorFromCurve(VectorCurve=DataInterface() {propertyValues {
       XCurve {keys [<at 0 3 RCIM_Cubic>, {arriveTangent=-8.3333, interpMode=RCIM_Cubic, leaveTangent=-8.3333, time=0.15, value=1}, <at 0.3 0.5 RCIM_Cubic>]}
       YCurve {keys [<at 0 3 RCIM_Cubic>, {arriveTangent=-8.3333, interpMode=RCIM_Cubic, leaveTangent=-8.3333, time=0.15, value=1}, <at 0.3 0.5 RCIM_Cubic>]}
       ZCurve {keys [<at 0 3 RCIM_Cubic>, {arriveTangent=-8.3333, interpMode=RCIM_Cubic, leaveTangent=-8.3333, time=0.15, value=1}, <at 0.3 0.5 RCIM_Cubic>]}
      }})
    ScaleA=true
    ScaleRGB=true
   )
  ]
  rendererClasses [&NiagaraRibbonRendererProperties]
  renderers [{rendererClass=&NiagaraRibbonRendererProperties}]
 }
}
summary {
 systemName=NS_GliderContrail_ValorL_Converted
 emitters [
  {emitterName=Trail_R, rendererClasses [&NiagaraRibbonRendererProperties]}
  {emitterName=Trail_R001, rendererClasses [&NiagaraRibbonRendererProperties]}
 ]
 userVariables [User.RibbonIDOffset {type=NiagaraInt32}]
}
