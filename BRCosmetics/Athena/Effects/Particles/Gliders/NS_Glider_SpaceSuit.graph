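# Niagara system for the SpaceSuit glider's trails: GreenTrail and a longer-lived, fainter OrangeTrail, each drawing two ribbons from skeletal mesh socket pairs.
# Socket midpoints set position, while socket separation sets ribbon width and facing.
# Each newest particle stays locked to its sockets until the next spawn, keeping ribbons attached.
# Spawning is distance-based, gated by User.ForwardDot; front-end preview adds rate spawning and drifts released particles to fake motion.
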NiagaraSystem /BRCosmetics/Athena/Effects/Particles/Gliders/NS_Glider_SpaceSuit
dependencies {usedDataInterfaces [&NiagaraDataInterfaceCurve], usedDynamicInputs [], usedModules [&SystemState]}
emitters {
 GreenTrail {
  emitterUpdateScript [
   &EmitterState()
   &SpawnPerUnit(
    Module.Use Max Movement Threshold=false
    Module.Use Movement Tolerance=false
    Spawn Probability:NiagaraFloat=&MakeCustomFloatFromBool(Bool=&SetBoolByFloatComparison(A=$User.ForwardDot, B=0.2), True Float=1, False Float=0)
    Use Spawn Probability=true
   )
   &SpawnRate(
    SpawnRate=&MakeCustomFloatFromBool(Bool=$User.bIsFrontEndPreview, True Float=60, False Float=0)
    Use Spawn Probability=false
   )
   /SetVariables(Emitter.HasSpawnedThisFrame:NiagaraBool=&Boolean_OrOperation(Bool1=$Output.SpawnPerUnit.HasSpawnedThisFrame, Bool2=$Output.SpawnRate.HasSpawnedThisFrame)) {isSetParameters=true}
  ]
  particleSpawnScript [
   &FN_SetSpawnGroup(Number of Spawn Groups=2)
   &InitializeParticle(
    Color=<rgba 1 1 1 0.5>
    Lifetime=0.15
    Sprite Size Mode=::NewEnumerator3
    Uniform Sprite Size=20
    UsePositionOffset=false
    Write Material Random=true
   )
   &SkeletalMeshLocation(
    Bone / Socket Index=&SpawnGroup()
    MaskSpawnGroup=false
    Mesh Sampling Type=::NewEnumerator5
    Position Sampling=::NewEnumerator2
    Skeletal Mesh=$User.SkeletalMesh
    Socket Sampling Mode=::NewEnumerator5
   )
   &SkeletalMeshLocation(
    Bone / Socket Index=&Add_Integer(A=$Particles.SpawnGroup, B=2)
    MaskSpawnGroup=false
    Mesh Sampling Type=::NewEnumerator5
    Position Sampling=::NewEnumerator2
    Skeletal Mesh=$User.SkeletalMesh
    Socket Sampling Mode=::NewEnumerator5
   ) {module=SkeletalMeshLocation001}
   /SetVariables(Particles.RibbonID=/Niagara/DynamicInputs/Meshes/MakeRibbonIDFromSpawnGroup()) {isSetParameters=true, module=SetVariables_#1}
   /SetVariables(Particles.Position=/Niagara/DynamicInputs/Lerp/Lerp_Position(
     A:NiagaraPosition=$Particles.SkeletalMeshLocation.SampledPosition
     B:NiagaraPosition=$Particles.SkeletalMeshLocation001.SampledPosition
     Alpha:NiagaraFloat=0.5
    )) {isSetParameters=true, module=SetVariables_#2}
   /SetVariables(Particles.RibbonWidth=&VectorLength(Vector=&Subtract_Vector(
      A=Linked() {linkedVariable {name=Particles.SkeletalMeshLocation.SampledPosition, type=NiagaraPosition}}
      B=$Particles.SkeletalMeshLocation001.SampledPosition
     ))) {isSetParameters=true, module=SetVariables_#3}
   /SetVariables(Particles.RibbonFacing=&NormalizeVector(
     Vector To Normalize=&Subtract_Vector(
      A=Linked() {linkedVariable {name=Particles.SkeletalMeshLocation.SampledPosition, type=NiagaraPosition}}
      B=$Particles.SkeletalMeshLocation001.SampledPosition
     )
     Safe Normalize=true
     Safe Fallback Threshold=0
     Safe Fallback Vector=<0 0 0.001>
    )) {isSetParameters=true, module=SetVariables_#4}
   /SetVariables(Particles.Locked=true) {isSetParameters=true}
  ]
  particleUpdateScript [
   &ParticleState()
   /SetVariables(Particles.Locked:NiagaraBool=&Boolean_AndOperation(
     Bool1=$Particles.Locked
     Bool2=&Boolean_OrOperation(Bool1=&Boolean_NotOperation(Bool=$Emitter.HasSpawnedThisFrame), Bool2=$Output.ParticleState.FirstFrame)
    )) {isSetParameters=true}
   &SkeletalMeshLocation(
    Bone / Socket Index=&SpawnGroup()
    MaskSpawnGroup=false
    Mesh Sampling Type=::NewEnumerator5
    Position Sampling=::NewEnumerator2
    Skeletal Mesh=$User.SkeletalMesh
    Socket Sampling Mode=::NewEnumerator5
   ) {module=SkeletalMeshLocation002}
   &SkeletalMeshLocation(
    Bone / Socket Index=&Add_Integer(A=&SpawnGroup(), B=2)
    MaskSpawnGroup=false
    Mesh Sampling Type=::NewEnumerator5
    Position Sampling=::NewEnumerator2
    Skeletal Mesh=$User.SkeletalMesh
    Socket Sampling Mode=::NewEnumerator5
   ) {module=SkeletalMeshLocation003}
   /SetVariables(Particles.RibbonFacing=&MakeCustomVectorFromBool(
     Bool=$Particles.Locked
     True Vector=&NormalizeVector(
      Vector To Normalize=&Subtract_Vector(A=$Particles.SkeletalMeshLocation002.SampledPosition, B=$Particles.SkeletalMeshLocation003.SampledPosition)
      Safe Normalize=true
      Safe Fallback Threshold=0
      Safe Fallback Vector=<0 0 0.001>
     )
     False Vector=$Particles.RibbonFacing
    )) {isSetParameters=true}
   /SetVariables(Particles.Position=&MakeCustomVectorFromBool(
     Bool=$Particles.Locked
     True Vector=&Lerp_Position(
      A:NiagaraPosition=Linked() {linkedVariable {name=Particles.SkeletalMeshLocation002.SampledPosition, type=Vector3f}}
      B:NiagaraPosition=Linked() {linkedVariable {name=Particles.SkeletalMeshLocation003.SampledPosition, type=Vector3f}}
      Alpha:NiagaraFloat=0.5
     )
     False Vector=&MakeCustomVectorFromBool(
      Bool=$User.bIsFrontEndPreview
      True Vector=&Add_Vector(
       A=Linked() {linkedVariable {name=Particles.Position, type=NiagaraPosition}}
       B=&Lerp_Vector(
        A=<0 0 0>
        B=&TransformVector(Vector=<0 -40 0>, Destination Space=::Simulation)
        Alpha=$Particles.NormalizedAge
       )
      )
      False Vector=Linked() {linkedVariable {name=Particles.Position, type=NiagaraPosition}}
     )
    )) {isSetParameters=true}
   &ScaleColor(
    Scale Alpha=&FloatFromCurve(FloatCurve=DataInterface() {propertyValues {Curve {keys [<at 0.5 1>,<at 1 0>]}}})
    ScaleA=true
    ScaleRGB=true
   )
  ]
  rendererClasses [&NiagaraRibbonRendererProperties]
  renderers [{rendererClass=&NiagaraRibbonRendererProperties}]
 }
 OrangeTrail {
  emitterUpdateScript [
   &EmitterState()
   &SpawnPerUnit(
    Module.Use Max Movement Threshold=false
    Module.Use Movement Tolerance=false
    Spawn Probability:NiagaraFloat=&MakeCustomFloatFromBool(Bool=&SetBoolByFloatComparison(A=$User.ForwardDot, B=0.2), True Float=1, False Float=0)
    Use Spawn Probability=true
   )
   &SpawnRate(
    SpawnRate=&MakeCustomFloatFromBool(Bool=$User.bIsFrontEndPreview, True Float=60, False Float=0)
    Use Spawn Probability=false
   )
   /SetVariables(Emitter.HasSpawnedThisFrame:NiagaraBool=&Boolean_OrOperation(Bool1=$Output.SpawnPerUnit.HasSpawnedThisFrame, Bool2=$Output.SpawnRate.HasSpawnedThisFrame)) {isSetParameters=true}
  ]
  particleSpawnScript [
   &FN_SetSpawnGroup(Number of Spawn Groups=2)
   &InitializeParticle(
    Color=<rgba 1 1 1 0.2>
    Lifetime=0.23
    Sprite Size Mode=::NewEnumerator3
    Uniform Sprite Size=20
    UsePositionOffset=false
    Write Material Random=true
   )
   &SkeletalMeshLocation(
    Bone / Socket Index=&SpawnGroup()
    MaskSpawnGroup=false
    Mesh Sampling Type=::NewEnumerator5
    Position Sampling=::NewEnumerator2
    Skeletal Mesh=$User.SkeletalMesh
    Socket Sampling Mode=::NewEnumerator5
   )
   &SkeletalMeshLocation(
    Bone / Socket Index=&Add_Integer(A=$Particles.SpawnGroup, B=2)
    MaskSpawnGroup=false
    Mesh Sampling Type=::NewEnumerator5
    Position Sampling=::NewEnumerator2
    Skeletal Mesh=$User.SkeletalMesh
    Socket Sampling Mode=::NewEnumerator5
   ) {module=SkeletalMeshLocation001}
   /SetVariables(Particles.RibbonID=&MakeRibbonIDFromSpawnGroup()) {isSetParameters=true, module=SetVariables_#1}
   /SetVariables(Particles.Position=&Lerp_Position(
     A:NiagaraPosition=$Particles.SkeletalMeshLocation.SampledPosition
     B:NiagaraPosition=$Particles.SkeletalMeshLocation001.SampledPosition
     Alpha:NiagaraFloat=0.5
    )) {isSetParameters=true, module=SetVariables_#2}
   /SetVariables(Particles.RibbonWidth=&VectorLength(Vector=&Subtract_Vector(
      A=Linked() {linkedVariable {name=Particles.SkeletalMeshLocation.SampledPosition, type=NiagaraPosition}}
      B=$Particles.SkeletalMeshLocation001.SampledPosition
     ))) {isSetParameters=true, module=SetVariables_#3}
   /SetVariables(Particles.RibbonFacing=&NormalizeVector(
     Vector To Normalize=&Subtract_Vector(
      A=Linked() {linkedVariable {name=Particles.SkeletalMeshLocation.SampledPosition, type=NiagaraPosition}}
      B=$Particles.SkeletalMeshLocation001.SampledPosition
     )
     Safe Normalize=true
     Safe Fallback Threshold=0
     Safe Fallback Vector=<0 0 0.001>
    )) {isSetParameters=true, module=SetVariables_#4}
   /SetVariables(Particles.Locked=true) {isSetParameters=true}
  ]
  particleUpdateScript [
   &ParticleState()
   &ScaleColor(
    Scale Alpha=&FloatFromCurve(FloatCurve=DataInterface() {propertyValues {Curve {keys [<at 0.5 1>,<at 1 0>]}}})
    ScaleA=true
    ScaleRGB=true
   )
   /SetVariables(Particles.Locked:NiagaraBool=&Boolean_AndOperation(
     Bool1=$Particles.Locked
     Bool2=&Boolean_OrOperation(Bool1=&Boolean_NotOperation(Bool=$Emitter.HasSpawnedThisFrame), Bool2=$Output.ParticleState.FirstFrame)
    )) {isSetParameters=true}
   &SkeletalMeshLocation(
    Bone / Socket Index=&SpawnGroup()
    MaskSpawnGroup=false
    Mesh Sampling Type=::NewEnumerator5
    Position Sampling=::NewEnumerator2
    Skeletal Mesh=$User.SkeletalMesh
    Socket Sampling Mode=::NewEnumerator5
   ) {module=SkeletalMeshLocation002}
   &SkeletalMeshLocation(
    Bone / Socket Index=&Add_Integer(A=&SpawnGroup(), B=2)
    MaskSpawnGroup=false
    Mesh Sampling Type=::NewEnumerator5
    Position Sampling=::NewEnumerator2
    Skeletal Mesh=$User.SkeletalMesh
    Socket Sampling Mode=::NewEnumerator5
   ) {module=SkeletalMeshLocation003}
   /SetVariables(Particles.RibbonFacing=&MakeCustomVectorFromBool(
     Bool=$Particles.Locked
     True Vector=&NormalizeVector(
      Vector To Normalize=&Subtract_Vector(A=$Particles.SkeletalMeshLocation002.SampledPosition, B=$Particles.SkeletalMeshLocation003.SampledPosition)
      Safe Normalize=true
      Safe Fallback Threshold=0
      Safe Fallback Vector=<0 0 0.001>
     )
     False Vector=$Particles.RibbonFacing
    )) {isSetParameters=true}
   /SetVariables(Particles.Position=&MakeCustomVectorFromBool(
     Bool=$Particles.Locked
     True Vector=&Lerp_Position(
      A:NiagaraPosition=Linked() {linkedVariable {name=Particles.SkeletalMeshLocation002.SampledPosition, type=Vector3f}}
      B:NiagaraPosition=Linked() {linkedVariable {name=Particles.SkeletalMeshLocation003.SampledPosition, type=Vector3f}}
      Alpha:NiagaraFloat=0.5
     )
     False Vector=&MakeCustomVectorFromBool(
      Bool=$User.bIsFrontEndPreview
      True Vector=&Add_Vector(
       A=Linked() {linkedVariable {name=Particles.Position, type=NiagaraPosition}}
       B=&Lerp_Vector(
        A=<0 0 0>
        B=&TransformVector(Vector=<0 -75 0>, Destination Space=::Simulation)
        Alpha=$Particles.NormalizedAge
       )
      )
      False Vector=Linked() {linkedVariable {name=Particles.Position, type=NiagaraPosition}}
     )
    )) {isSetParameters=true}
  ]
  rendererClasses [&NiagaraRibbonRendererProperties]
  renderers [{rendererClass=&NiagaraRibbonRendererProperties}]
 }
}
summary {
 systemName=NS_Glider_SpaceSuit
 emitters [
  {emitterName=GreenTrail, rendererClasses [&NiagaraRibbonRendererProperties]}
  {emitterName=OrangeTrail, rendererClasses [&NiagaraRibbonRendererProperties]}
 ]
 userVariables [
  User.bIsFrontEndPreview {type=NiagaraBool}
  User.ForwardDot {type=NiagaraFloat, defaultValue=0, description=""}
  User.Movement {type=NiagaraFloat, defaultValue=0, description=""}
  User.RibbonIDOffset {type=NiagaraInt32}
  User.SkeletalMesh {
   type=NiagaraDataInterfaceSkeletalMesh
   defaultValue {
    struct=NiagaraExt_VariableValue_DataInterface
    value {dataInterface=&self:NiagaraDataInterfaceSkeletalMesh_0}
   }
  }
 ]
}
