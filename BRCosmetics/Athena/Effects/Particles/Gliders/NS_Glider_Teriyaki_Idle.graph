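# Niagara idle effect for the Teriyaki fish glider that sheds blue sprite drops and splashes from the attached glider mesh.
# Both emitters use Skeletal Mesh Location on the skinned parent, filtered to the same left and right branch bones.
# Drops spawn at a waveform-modulated rate, fall and shrink, with sizes ramping up over system age.
# Splashes spawn steadily, grow and spin under curve-ramped gravity, darkening and fading with a slight camera offset.
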NiagaraSystem /BRCosmetics/Athena/Effects/Particles/Gliders/NS_Glider_Teriyaki_Idle
dependencies {
 usedDataInterfaces [&NiagaraDataInterfaceCurve, &NiagaraDataInterfaceColorCurve]
 usedDynamicInputs []
 usedModules [&SystemState]
}
emitters {
 Drops {
  emitterUpdateScript [
   &EmitterState()
   &SpawnRate(
    SpawnRate=&Waveform(
     Global Amplitude Scale=0.5
     "Waveform [1]"=::NewEnumerator0
     "[1] Frequency"=1
     "[1] Amplitude Min/Max"=<0 1>
     "[1] Scale"=100
     "[1] Bias"=50
    )
    Use Spawn Probability=false
   )
  ]
  particleSpawnScript [
   &InitializeParticle(
    Color=<rgba 0.2865 0.7862 3 1>
    Lifetime=0.5
    Sprite Size Max=<5 18>
    Sprite Size Min=<4 7>
    Sprite Size Mode=::NewEnumerator2
    UsePositionOffset=false
    Write Material Random=true
   )
   &AddVelocity(
    Cone Angle=90
    Cone Axis=<0 1 1>
    Module.UseConeFalloff=false
    Module.UseVelDistribution=false
    Override Seed=false
    Velocity Mode=::NewEnumerator2
    Velocity Speed=150
   )
   &ApplyInitialForces()
   &SkeletalMeshLocation(
    Bone Sampling Mode=::NewEnumerator0
    Interpolate Mesh Sampling=false
    MaskSpawnGroup=false
    OverrideRandomSeed=false
    Skeletal Mesh=DataInterface() {propertyValues {
      DefaultMesh=null
      ExcludeBoneName=root
      OptionalComponentClass=null
      PreviewMesh=/BRCosmetics/Gadgets/Assets/VinderTech_GliderChute/Glider_TeriyakiFish/Meshes/Glider_TeriyakiFish
      SkinningMode=SkinOnTheFly
      SoftSourceActor=null
      SourceMode=AttachParent
      UvSetIndex=0
      WholeMeshLOD=-1
      bExcludeBone=false
      bReadDeformedGeometry=true
      bRequireCurrentFrameData=true
      ComponentTags []
      FilteredBones [
       leafBranchA3_03_l
       leafBranchA4_03_l
       leafBranchA2_05_l
       leafBranchA1_03_l
       mainBranchB_07_r
       leafBranchB2_04_r
       leafBranchA4_08_r
       leafBranchA3_010_r
       mainBranchA_09_r
       leafBranchB1_06_l
       leafBranchB2_04_l
       leafBranchB3_04_l
       leafBranchB4_03_l
       mainBranchB_012_l
       mainBranchA_12_l
       leafBranchB1_03_r
      ]
      FilteredSockets []
      MeshUserParameter {parameter {name=None, typeDefHandle {registeredTypeIndex=1}, varData []}}
      SamplingRegions []
     }}
   )
  ]
  particleUpdateScript [
   &ParticleState()
   &ScaleSpriteSize(
    Uniform Curve Scale=&FloatFromCurve(
     FloatCurve=DataInterface() {propertyValues {Curve {keys [
         {arriveTangent=0.2723, interpMode=RCIM_Cubic, leaveTangent=0.2723, tangentMode=RCTM_User, time=0, value=0}
         {arriveTangent=0.5056, tangentMode=RCTM_User, time=2, value=1}
        ]}}}
     CurveIndex=$System.Age
    )
    Uniform Curve Sprite Scale=DataInterface() {propertyValues {Curve {
       defaultValue=0
       keys [
        {arriveTangent=-2.1609, interpMode=RCIM_Cubic, leaveTangent=-2.1609, tangentMode=RCTM_User, time=0, value=1}
        {arriveTangent=-4.0358, interpMode=RCIM_Cubic, leaveTangent=-4.0358, tangentMode=RCTM_User, time=1, value=0}
       ]
      }}}
   )
   &GravityForce(Gravity=<0 0 -700>)
   &SolveForcesAndVelocity(Clamp Velocity=false, Limit Acceleration=false)
  ]
  rendererClasses [&NiagaraSpriteRendererProperties]
  renderers [{rendererClass=&NiagaraSpriteRendererProperties}]
 }
 Splashes {
  emitterUpdateScript [
   &EmitterState()
   &SpawnRate(SpawnRate=80, Use Spawn Probability=false)
  ]
  particleSpawnScript [
   &InitializeParticle(
    Color Mode=::NewEnumerator0
    Lifetime Max=0.75
    Lifetime Min=0.5
    Lifetime Mode=::NewEnumerator1
    Sprite Rotation Mode=::NewEnumerator2
    Sprite Size Mode=::NewEnumerator1
    Sprite UV Mode=::NewEnumerator3
    Uniform Sprite Size Max=30
    Uniform Sprite Size Min=15
    UsePositionOffset=false
    Write Material Random=true
   )
   &AddVelocity(
    Cone Axis=<0 0.2 1>
    Module.UseConeFalloff=false
    Module.UseVelDistribution=false
    Override Seed=false
    Velocity Mode=::NewEnumerator2
    Velocity Speed=20
   )
   &ApplyInitialForces()
   &SkeletalMeshLocation(
    Bone Sampling Mode=::NewEnumerator0
    Interpolate Mesh Sampling=false
    MaskSpawnGroup=false
    OverrideRandomSeed=false
    Sampled Scale Coordinate Space:ENiagaraCoordinateSpace=::Simulation
    Skeletal Mesh=DataInterface() {propertyValues {
      DefaultMesh=null
      ExcludeBoneName=root
      OptionalComponentClass=null
      PreviewMesh=&Glider_TeriyakiFish
      SkinningMode=SkinOnTheFly
      SoftSourceActor=null
      SourceMode=AttachParent
      UvSetIndex=0
      WholeMeshLOD=-1
      bExcludeBone=false
      bReadDeformedGeometry=true
      bRequireCurrentFrameData=true
      ComponentTags []
      FilteredBones [
       leafBranchA3_03_l
       leafBranchA4_03_l
       leafBranchA2_05_l
       leafBranchA1_03_l
       mainBranchB_07_r
       leafBranchB2_04_r
       leafBranchA4_08_r
       leafBranchA3_010_r
       mainBranchA_09_r
       leafBranchB1_06_l
       leafBranchB2_04_l
       leafBranchB3_04_l
       leafBranchB4_03_l
       mainBranchB_012_l
       mainBranchA_12_l
       leafBranchB1_03_r
      ]
      FilteredSockets []
      MeshUserParameter {parameter {name=None, typeDefHandle {registeredTypeIndex=1}, varData []}}
      SamplingRegions []
     }}
   )
  ]
  particleUpdateScript [
   &ParticleState()
   &ScaleSpriteSize(Uniform Curve Sprite Scale=DataInterface() {propertyValues {Curve {keys [
        {arriveTangent=1.4, interpMode=RCIM_Cubic, leaveTangent=1.4, tangentMode=RCTM_User, time=0, value=0}
        {arriveTangent=1.8929, tangentMode=RCTM_User, time=1.0038, value=1.5376}
       ]}}})
   &SpriteRotationRate(Rotation Rate=&RandomRangeFloat(Minimum=-100, Maximum=100))
   &GravityForce(Gravity=&Multiply_VectorByFloat(
     Vector=<0 100 -200>
     Float=&FloatFromCurve(FloatCurve=DataInterface() {propertyValues {Curve {keys [{arriveTangent=4.4489, interpMode=RCIM_Cubic, leaveTangent=4.4489, tangentMode=RCTM_User, time=0, value=0}, <at 0.5 1>]}}})
    ))
   &SolveForcesAndVelocity(Clamp Velocity=false, Limit Acceleration=false)
   &Color(Color=&ColorFromCurve(ColorCurve=DataInterface() {propertyValues {
       AlphaCurve {keys [
         <at 0.0063 0>
         <at 0.1867 2>
         {arriveTangent=-2.4591, interpMode=RCIM_Cubic, leaveTangent=-2.4591, time=0.5969, value=1}
         <at 1 0 RCIM_Cubic>
        ]}
       BlueCurve {keys [<at 0.0063 1 RCIM_Cubic>, {interpMode=RCIM_Cubic, leaveTangent=-0.7875, time=0.1252, value=0.2274}]}
       GreenCurve {keys [<at 0.0063 0.7234 RCIM_Cubic>, {interpMode=RCIM_Cubic, leaveTangent=-0.675, time=0.1252, value=0.0612}]}
       RedCurve {keys [<at 0.0063 0.4688 RCIM_Cubic>, {interpMode=RCIM_Cubic, leaveTangent=-0.466, time=0.1252, value=0.0116}]}
      }}))
   &CameraOffset(Camera Offset Amount=5)
  ]
  rendererClasses [&NiagaraSpriteRendererProperties]
  renderers [{rendererClass=&NiagaraSpriteRendererProperties}]
 }
}
summary {
 systemName=NS_Glider_Teriyaki_Idle
 emitters [
  {emitterName=Drops, rendererClasses [&NiagaraSpriteRendererProperties]}
  {emitterName=Splashes, rendererClasses [&NiagaraSpriteRendererProperties]}
 ]
 userVariables [User.SkeletalMesh {
   type=NiagaraDataInterfaceSkeletalMesh
   defaultValue {
    struct=NiagaraExt_VariableValue_DataInterface
    value {dataInterface=&self:NiagaraDataInterfaceSkeletalMesh_0}
   }
  }]
}
