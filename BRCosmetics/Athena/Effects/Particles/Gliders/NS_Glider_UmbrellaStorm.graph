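# Niagara effect for the Umbrella Storm glider cosmetic, built from three mesh-particle emitters that follow the User.SkeletalMesh.
# Start plays a one-second intro by animating material parameters, then a delayed Loop emitter sustains the look with long-lived meshes slowly rotated around Z using system age.
# Lightning_Arcs repeatedly spawns short-lived, randomly oriented bolts at randomly chosen bones or sockets, using curves for flickering alpha, mesh scale and dynamic material parameters.
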NiagaraSystem /BRCosmetics/Athena/Effects/Particles/Gliders/NS_Glider_UmbrellaStorm
dependencies {
 usedDataInterfaces [&NiagaraDataInterfaceCurve, &NiagaraDataInterfaceArrayFloat3, &NiagaraDataInterfaceVectorCurve]
 usedDynamicInputs []
 usedModules [&SystemState]
}
emitters {
 Lightning_Arcs {
  emitterUpdateScript [
   &EmitterState(
    Delay First Loop Only=true
    Life Cycle Mode=::NewEnumerator1
    Loop Delay=1
    Loop Duration=&RandomRangeFloat(Minimum=0.1, Maximum=0.8)
    Recalculate Duration Each Loop=true
    UseLoopDelay=true
   )
   &SpawnRate(SpawnRate=1.5, Use Spawn Probability=false)
   &SpawnBurst_Instantaneous(Use Spawn Probability=false, UseLoopCountLimit=false)
  ]
  particleSpawnScript [
   &InitializeParticle(
    Color=<rgba 1.75 1.8 1.25 1>
    Lifetime=0.25
    Mesh Scale Max=<0.13 0.25 0.33>
    Mesh Scale Min=<0.1 0.2 0.25>
    Mesh Scale Mode=::NewEnumerator2
    UsePositionOffset=false
    Write Material Random=true
   )
   /SetVariables(Particles.RandomInt:NiagaraInt32=&RandomRangeInt(Minimum=0, Maximum=2)) {isSetParameters=true}
   &AddVelocity(Velocity=&RandomRangeVector(Minimum=<-0.1 -1 0.75>, Maximum=<0.1 0 1>))
   &SkeletalMeshLocation(
    Bone / Socket Index=$Particles.RandomInt
    MaskSpawnGroup=false
    Mesh Sampling Type=::NewEnumerator5
    Sampled Position Offset=&SelectVectorFromArray(Direct Array Index=$Particles.RandomInt, Vector Selection Array=DataInterface() {propertyValues {FloatData [<0 0 0>, <0 0 0>, <-35 0 0>]}})
    Skeletal Mesh=$User.SkeletalMesh
    Socket Sampling Mode=::NewEnumerator5
   )
   &InitialMeshOrientation(
    Rotation=&RandomRangeVector(Minimum=<0 0 0>, Maximum=<1 1 1>)
    Use Orientation Vector=false
    Use Rotation Vector=true
   )
  ]
  particleUpdateScript [
   &ParticleState()
   &ScaleColor(
    Scale Alpha=&FloatFromCurve(FloatCurve=DataInterface() {propertyValues {Curve {keys [
         <at 0.4 1 RCIM_Cubic>
         {arriveTangent=-2.8077, interpMode=RCIM_Cubic, leaveTangent=-2.8077, time=0.4925, value=1.12}
         {arriveTangent=-1.9653, interpMode=RCIM_Cubic, leaveTangent=-1.9653, time=0.53, value=0.635}
         {arriveTangent=-2.3333, interpMode=RCIM_Cubic, leaveTangent=-2.3333, time=0.579, value=0.95}
         {arriveTangent=2.3596, interpMode=RCIM_Cubic, leaveTangent=2.3596, time=0.62, value=0.425}
         {arriveTangent=-1.95, interpMode=RCIM_Cubic, leaveTangent=-1.95, time=0.668, value=1.16}
         {arriveTangent=-4.7826, interpMode=RCIM_Cubic, leaveTangent=-4.7826, time=0.72, value=0.23}
         {arriveTangent=-1.2111, interpMode=RCIM_Cubic, leaveTangent=-1.2111, time=0.76, value=0.72}
         {arriveTangent=0.4124, interpMode=RCIM_Cubic, leaveTangent=0.4124, time=0.81, value=0.121}
         {arriveTangent=-0.8526, interpMode=RCIM_Cubic, leaveTangent=-0.8526, time=0.857, value=0.76}
         {arriveTangent=-1.7204, interpMode=RCIM_Cubic, leaveTangent=-1.7204, time=0.905, value=0.04}
         {arriveTangent=-0.4211, interpMode=RCIM_Cubic, leaveTangent=-0.4211, time=0.95, value=0.6}
         {arriveTangent=-12, interpMode=RCIM_Cubic, leaveTangent=-12, time=1, value=0}
         <at 1.05 -0.6 RCIM_Cubic>
        ]}}})
    ScaleA=true
    ScaleRGB=true
   )
   &ScaleMeshSize(Scale Factor=&VectorFromCurve(VectorCurve=DataInterface() {propertyValues {
       XCurve {keys [
         <at 0 0.95>
         {arriveTangent=10.6834, interpMode=RCIM_Cubic, leaveTangent=10.6834, tangentMode=RCTM_User, time=0.02, value=-1.15}
         {arriveTangentWeight=0.3932, interpMode=RCIM_Cubic, leaveTangentWeight=0.4096, tangentMode=RCTM_User, time=0.5, value=1.07}
         <at 1 1.05>
        ]}
       YCurve {keys [
         <at 0 1>
         {arriveTangent=10.6834, interpMode=RCIM_Cubic, leaveTangent=10.6834, tangentMode=RCTM_User, time=0.02, value=-1.03}
         {arriveTangentWeight=0.368, interpMode=RCIM_Cubic, leaveTangentWeight=0.3834, tangentMode=RCTM_User, time=0.5, value=1}
         <at 1 1>
        ]}
       ZCurve {keys [
         <at 0 1>
         {arriveTangent=14.9802, interpMode=RCIM_Cubic, leaveTangent=14.9802, tangentMode=RCTM_User, time=0.02, value=-5.09}
         {arriveTangentWeight=1.0071, interpMode=RCIM_Cubic, leaveTangentWeight=1.049, tangentMode=RCTM_User, time=0.5, value=1}
         <at 1 1>
        ]}
      }}))
   &DynamicMaterialParameters(
    Index 0 Param 1=&FloatFromCurve(FloatCurve=DataInterface() {propertyValues {Curve {keys [
         {arriveTangent=201.4994, interpMode=RCIM_Cubic, leaveTangent=201.4994, tangentMode=RCTM_User, time=0, value=0}
         {arriveTangentWeight=0.0178, interpMode=RCIM_Cubic, leaveTangentWeight=0.3209, tangentMode=RCTM_User, time=0.05, value=4.6}
         <at 0.95 2.5 RCIM_Cubic>
        ]}}})
    Index 0 Param 2=&FloatFromCurve(FloatCurve=DataInterface() {propertyValues {Curve {keys [
         <at 0 1>
         {arriveTangent=2, leaveTangent=2, time=0.1, value=0.9}
         {arriveTangent=0.8584, leaveTangent=0.8584, time=0.25, value=1.5}
         {arriveTangent=-2.2, leaveTangent=-2.2, time=0.333, value=1.1}
         {arriveTangent=-0.8996, interpMode=RCIM_Cubic, leaveTangent=-0.8996, time=0.5, value=0.95}
         <at 1 0.5 RCIM_Cubic>
        ]}}})
    Index 0 Param 3=&RandomRangeFloat(Minimum=0, Maximum=1)
    Index 0 Param 4=&RandomRangeFloat(Minimum=0, Maximum=1)
    Module.Param0WriteEnabled=true
    Module.Param1WriteEnabled=true
    Module.Param2WriteEnabled=true
    Module.Param3WriteEnabled=true
   )
   &SolveForcesAndVelocity(Clamp Velocity=false, Limit Acceleration=false)
  ]
  rendererClasses [&NiagaraMeshRendererProperties]
  renderers [{rendererClass=&NiagaraMeshRendererProperties, rendererIndex=0}]
 }
 Loop {
  emitterUpdateScript [
   &EmitterState(
    Delay First Loop Only=true
    Inactive Response=::NewEnumerator1
    Life Cycle Mode=::NewEnumerator1
    Loop Delay=0.8
    Loop Duration=10
    UseLoopDelay=true
   )
   &SpawnBurst_Instantaneous(Use Spawn Probability=false, UseLoopCountLimit=false)
  ]
  particleSpawnScript [
   &InitializeParticle(Color Mode=::NewEnumerator0, Lifetime=10, UsePositionOffset=false, Write Material Random=true)
   &DynamicMaterialParameters(
    Index 0 Param 1=1
    Index 0 Param 2=3
    Index 0 Param 3=1
    Index 0 Param 4=1
    Module.Param0WriteEnabled=true
    Module.Param1WriteEnabled=true
    Module.Param2WriteEnabled=true
    Module.Param3WriteEnabled=true
   )
  ]
  particleUpdateScript [
   &ParticleState()
   &SkeletalMeshLocation(Bone Sampling Mode=::NewEnumerator4, MaskSpawnGroup=false, Skeletal Mesh=$User.SkeletalMesh)
   &UpdateMeshOrientation(Delta Time:NiagaraFloat=$Engine.System.Age, Rotation Coordinate Space=::NewEnumerator3, Rotation Rate=0.1, Rotation Vector=<0 0 1>)
  ]
  rendererClasses [&NiagaraMeshRendererProperties]
  renderers [
   {rendererClass=&NiagaraMeshRendererProperties, rendererIndex=0}
   {rendererClass=&NiagaraMeshRendererProperties, rendererIndex=1}
  ]
 }
 Start {
  emitterUpdateScript [
   &EmitterState(Inactive Response=::NewEnumerator1, Life Cycle Mode=::NewEnumerator1, Loop Behavior=::NewEnumerator1, Loop Duration=1, UseLoopDelay=false)
   &SpawnBurst_Instantaneous(Use Spawn Probability=false, UseLoopCountLimit=false)
  ]
  particleSpawnScript [&InitializeParticle(Color Mode=::NewEnumerator0, UsePositionOffset=false, Write Material Random=true)]
  particleUpdateScript [
   &ParticleState()
   &SkeletalMeshLocation(Bone Sampling Mode=::NewEnumerator4, MaskSpawnGroup=false, Skeletal Mesh=$User.SkeletalMesh)
   &DynamicMaterialParameters(
    Index 0 Param 1=&FloatFromCurve(FloatCurve=DataInterface() {propertyValues {Curve {keys [{tangentMode=RCTM_User, time=0, value=0}, {tangentMode=RCTM_User, time=1, value=1}]}}})
    Index 0 Param 2=&FloatFromCurve(FloatCurve=DataInterface() {propertyValues {Curve {keys [<at 0 100>,<at 1 3>]}}})
    Index 0 Param 3=1
    Index 0 Param 4=1
    Module.Param0WriteEnabled=true
    Module.Param1WriteEnabled=true
    Module.Param2WriteEnabled=true
    Module.Param3WriteEnabled=true
   )
  ]
  rendererClasses [&NiagaraMeshRendererProperties]
  renderers [
   {rendererClass=&NiagaraMeshRendererProperties, rendererIndex=0}
   {rendererClass=&NiagaraMeshRendererProperties, rendererIndex=1}
  ]
 }
}
summary {
 systemName=NS_Glider_UmbrellaStorm
 emitters [
  {emitterName=Start, rendererClasses [&NiagaraMeshRendererProperties]}
  {emitterName=Loop, rendererClasses [&NiagaraMeshRendererProperties]}
  {emitterName=Lightning_Arcs, rendererClasses [&NiagaraMeshRendererProperties]}
 ]
 userVariables [User.SkeletalMesh {
   type=NiagaraDataInterfaceSkeletalMesh
   defaultValue {
    struct=NiagaraExt_VariableValue_DataInterface
    value {dataInterface=&self:NiagaraDataInterfaceSkeletalMesh_0}
   }
  }]
}
