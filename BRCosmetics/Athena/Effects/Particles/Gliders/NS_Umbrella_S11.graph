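# Niagara effect for the S11 umbrella glider, styled as water that splashes open into a sustained, dripping canopy.
# One-shot emitters play the opening: a scaling mesh, rotating inner and outer wash meshes, splash arcs and drip bursts.
# Delayed looping emitters then sustain the canopy mesh and drips from its rim and top.
# Meshes animate through dynamic material parameters, while drips are GPU sprites with gravity and speed-based stretching.
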NiagaraSystem /BRCosmetics/Athena/Effects/Particles/Gliders/NS_Umbrella_S11
dependencies {
 usedDataInterfaces [
  &NiagaraDataInterfaceVectorCurve
  &NiagaraDataInterfaceCurve
  &NiagaraDataInterfaceVector2DCurve
 ]
 usedDynamicInputs []
 usedModules [&SystemState]
}
emitters {
 DripsLoop {
  simTarget=GPUComputeSim
  emitterUpdateScript [
   &EmitterState(Delay First Loop Only=true, Life Cycle Mode=::NewEnumerator1, Loop Delay=1.25, Loop Duration=1, UseLoopDelay=true)
   &SpawnRate(SpawnRate=200, Use Spawn Probability=false)
  ]
  particleSpawnScript [
   &InitializeParticle(
    Lifetime=&RandomRangeFloat(Minimum=0.4, Maximum=0.7, Evaluation Type=::NewEnumerator1)
    Sprite Size=&RandomRangeVector2D(Minimum=<0.6 2.5>, Maximum=<2 4>, Evaluation Type=::NewEnumerator1)
    Sprite Size Mode=::NewEnumerator4
    UsePositionOffset=false
    Write Material Random=true
   )
   &Color(Color=&MakeLinearColorFromVectorAndFloat("Vector (RGB)"=<1 1 1>, "Float (Alpha)"=3))
   &AddVelocity(
    Rotation Coordinate Space=::World
    Velocity=&RandomRangeVector(Minimum=<-40 -40 50>, Maximum=<40 40 100>, Evaluation Type=::NewEnumerator1)
   )
   &ShapeLocation(
    Cylinder Height=8
    Cylinder Radius=&RandomRangeFloat(Minimum=48, Maximum=52, Evaluation Type=::NewEnumerator1)
    MaskSpawnGroup=false
    Override Seed=false
    Rotation Mode=::NewEnumerator2
    Shape Primitive=::NewEnumerator1
    Surface Only=true
    Use Endcaps In Surface Only Mode=false
    UseOwnerScale=true
   )
   /CascadeToNiagaraConverter/NiagaraScripts/ModuleScripts/AddVectorToPosition(Vector:Vector3f=&TransformVector(Vector=<0 0 95>, Destination Space=::Simulation))
   &InheritVelocity(Inherited Velocity Amount Scale=<-0.15 -0.15 -0.15>, Inherited Velocity Speed Limit=0, Speed Limit=false, UseSpeedThreshold=false)
  ]
  particleUpdateScript [
   &ParticleState()
   &ScaleSpriteSize(
    Scale Factor=&Vector2DFromFloat(Value=&FloatFromCurve(FloatCurve=DataInterface() {propertyValues {Curve {keys [<at 0 1>,<at 1 0.4>]}}}))
    Scale Sprite Size Mode=::NewEnumerator2
   )
   &ScaleColor(
    Scale Alpha=&FloatFromCurve(FloatCurve=DataInterface() {propertyValues {Curve {keys [<at 0 0 RCIM_Cubic>,<at 0.25 1 RCIM_Cubic>,<at 1 0 RCIM_Cubic>]}}})
    Scale RGB=&VectorFromCurve(VectorCurve=DataInterface() {propertyValues {
       XCurve {keys [<at 0.25 0.9062 RCIM_Cubic>,<at 0.85 0.333>]}
       YCurve {keys [<at 0.25 1 RCIM_Cubic>,<at 0.85 0.5274>]}
       ZCurve {keys [<at 0.25 0.9468 RCIM_Cubic>,<at 0.85 0.6>]}
      }})
    ScaleA=true
    ScaleRGB=true
   )
   &AccelerationForce(Acceleration=<0 0 -900>)
   &SolveForcesAndVelocity(Clamp Velocity=false, Limit Acceleration=false)
   &ScaleSpriteSizeBySpeed(Max Scale Factor=<10 20>, Source Velocity:Vector3f=&Multiply_Vector(A=<0 0.05 1>, B=$Particles.Velocity))
  ]
  rendererClasses [&NiagaraSpriteRendererProperties]
  renderers [{rendererClass=&NiagaraSpriteRendererProperties}]
 }
 DripsLoopTop {
  simTarget=GPUComputeSim
  emitterUpdateScript [
   &EmitterState(Delay First Loop Only=true, Life Cycle Mode=::NewEnumerator1, Loop Delay=0.8, Loop Duration=1, UseLoopDelay=true)
   &SpawnRate(SpawnRate=150, Use Spawn Probability=false)
  ]
  particleSpawnScript [
   &InitializeParticle(
    Lifetime=&RandomRangeFloat(Minimum=0.35, Maximum=0.64, Evaluation Type=::NewEnumerator1)
    Sprite Size=&RandomRangeVector2D(Minimum=<0.6 2.5>, Maximum=<2 4>, Evaluation Type=::NewEnumerator1)
    Sprite Size Mode=::NewEnumerator4
    UsePositionOffset=false
    Write Material Random=true
   )
   &Color(Color=&MakeLinearColorFromVectorAndFloat("Vector (RGB)"=<1 1 1>, "Float (Alpha)"=3))
   &AddVelocity(
    Rotation Coordinate Space=::World
    Velocity=&RandomRangeVector(Minimum=<-20 -20 120>, Maximum=<20 20 180>, Evaluation Type=::NewEnumerator1)
   )
   &ShapeLocation(
    Cylinder Height=8
    Cylinder Radius=&RandomRangeFloat(Minimum=4, Maximum=5, Evaluation Type=::NewEnumerator1)
    MaskSpawnGroup=false
    Override Seed=false
    Rotation Mode=::NewEnumerator2
    Shape Primitive=::NewEnumerator1
    Surface Only=true
    Use Endcaps In Surface Only Mode=false
    UseOwnerScale=true
   )
   /CascadeToNiagaraConverter/NiagaraScripts/ModuleScripts/FixShapeVector()
   &AddVelocity(Velocity=&Multiply_VectorByFloat(Vector=&Multiply_Vector(A=$Output.FixShapeVector.ShapeVector, B=<1 1 0>), Float=25)) {module=AddVelocity001}
   &AddVectorToPosition(Vector:Vector3f=&TransformVector(Vector=<0 0 110>, Destination Space=::Simulation))
   &InheritVelocity(Inherited Velocity Amount Scale=<-0.05 -0.05 0>, Inherited Velocity Speed Limit=0, Speed Limit=false, UseSpeedThreshold=false)
  ]
  particleUpdateScript [
   &ParticleState()
   &ScaleSpriteSize(
    Scale Factor=&Vector2DFromFloat(Value=&FloatFromCurve(FloatCurve=DataInterface() {propertyValues {Curve {keys [<at 0 1>,<at 1 0.4>]}}}))
    Scale Sprite Size Mode=::NewEnumerator2
   )
   &ScaleColor(
    Scale Alpha=&FloatFromCurve(FloatCurve=DataInterface() {propertyValues {Curve {keys [
         {arriveTangent=11.6133, interpMode=RCIM_Cubic, leaveTangent=11.6133, tangentMode=RCTM_User, time=0, value=0}
         <at 0.15 1 RCIM_Cubic>
         {arriveTangent=-2.6975, interpMode=RCIM_Cubic, leaveTangent=-2.6975, tangentMode=RCTM_User, time=1, value=0}
        ]}}})
    Scale RGB=&VectorFromCurve(VectorCurve=DataInterface() {propertyValues {
       XCurve {keys [<at 0.25 0.9062 RCIM_Cubic>,<at 0.85 0.333>]}
       YCurve {keys [<at 0.25 1 RCIM_Cubic>,<at 0.85 0.5274>]}
       ZCurve {keys [<at 0.25 0.9468 RCIM_Cubic>,<at 0.85 0.6>]}
      }})
    ScaleA=true
    ScaleRGB=true
   )
   &AccelerationForce(Acceleration=<0 0 -900>)
   &SolveForcesAndVelocity(Clamp Velocity=false, Limit Acceleration=false)
   &ScaleSpriteSizeBySpeed(Max Scale Factor=<10 20>, Source Velocity:Vector3f=&Multiply_Vector(A=<0 0.05 1>, B=$Particles.Velocity))
  ]
  rendererClasses [&NiagaraSpriteRendererProperties]
  renderers [{rendererClass=&NiagaraSpriteRendererProperties}]
 }
 DripsStart {
  simTarget=GPUComputeSim
  emitterUpdateScript [
   &EmitterState(Life Cycle Mode=::NewEnumerator1, Loop Behavior=::NewEnumerator1, Loop Delay=0.9, Loop Duration=1, UseLoopDelay=true)
   &SpawnBurst_Instantaneous(Spawn Count=45, Spawn Time=0.04, Use Spawn Probability=false, UseLoopCountLimit=false) {module=SpawnBurst_Instantaneous001}
   &SpawnBurst_Instantaneous(Spawn Count=55, Spawn Time=0.01, Use Spawn Probability=false, UseLoopCountLimit=false)
   &SpawnRate(
    SpawnRate=&Multiply_Float(
     A=300
     B=&FloatFromCurve(FloatCurve=DataInterface() {propertyValues {Curve {keys [<at 0 1 RCIM_Cubic>,<at 1 0>]}}}, CurveIndex=$Emitter.NormalizedLoopAge)
    )
    Use Spawn Probability=false
   )
  ]
  particleSpawnScript [
   &InitializeParticle(
    Lifetime=&RandomRangeFloat(Minimum=0.5, Maximum=1, Evaluation Type=::NewEnumerator1)
    Sprite Size=&RandomRangeVector2D(Minimum=<0.6 2.5>, Maximum=<2 4>, Evaluation Type=::NewEnumerator1)
    Sprite Size Mode=::NewEnumerator4
    UsePositionOffset=false
    Write Material Random=true
   )
   &Color(Color=&MakeLinearColorFromVectorAndFloat("Vector (RGB)"=<1 1 1>, "Float (Alpha)"=3))
   &AddVelocity(
    Rotation Coordinate Space=::World
    Velocity=&RandomRangeVector(Minimum=<-40 -40 100>, Maximum=<40 40 400>, Evaluation Type=::NewEnumerator1)
   )
   &ShapeLocation(
    Cylinder Height=5
    Cylinder Radius=&RandomRangeFloat(Minimum=30, Maximum=40, Evaluation Type=::NewEnumerator1)
    MaskSpawnGroup=false
    Override Seed=false
    Rotation Mode=::NewEnumerator2
    Shape Primitive=::NewEnumerator1
    Surface Only=false
    UseOwnerScale=true
   )
   &FixShapeVector()
   &AddVelocity(Velocity=&Multiply_VectorByFloat(Vector=&Multiply_Vector(A=$Output.FixShapeVector.ShapeVector, B=<1 1 0>), Float=2)) {module=AddVelocity001}
   &AddVectorToPosition(Vector:Vector3f=&TransformVector(Vector=<0 0 100>, Destination Space=::Simulation))
  ]
  particleUpdateScript [
   &ParticleState()
   &ScaleSpriteSize(
    Scale Factor=&Vector2DFromFloat(Value=&FloatFromCurve(FloatCurve=DataInterface() {propertyValues {Curve {keys [<at 0 1>,<at 1 0.4>]}}}))
    Scale Sprite Size Mode=::NewEnumerator2
   )
   &ScaleColor(
    Scale Alpha=&FloatFromCurve(FloatCurve=DataInterface() {propertyValues {Curve {keys [<at 0 0 RCIM_Cubic>,<at 0.25 1 RCIM_Cubic>,<at 1 0 RCIM_Cubic>]}}})
    Scale RGB=&VectorFromCurve(VectorCurve=DataInterface() {propertyValues {
       XCurve {keys [<at 0.25 0.9062 RCIM_Cubic>,<at 0.85 0.333>]}
       YCurve {keys [<at 0.25 1 RCIM_Cubic>,<at 0.85 0.5274>]}
       ZCurve {keys [<at 0.25 0.9468 RCIM_Cubic>,<at 0.85 0.6>]}
      }})
    ScaleA=true
    ScaleRGB=true
   )
   &AccelerationForce(Acceleration=<0 0 -1200>)
   &SolveForcesAndVelocity(Clamp Velocity=false, Limit Acceleration=false)
   &ScaleSpriteSizeBySpeed(Max Scale Factor=<10 20>, Source Velocity:Vector3f=&Multiply_Vector(A=<0 0.05 1>, B=$Particles.Velocity))
  ]
  rendererClasses [&NiagaraSpriteRendererProperties]
  renderers [{rendererClass=&NiagaraSpriteRendererProperties}]
 }
 DripsStartTop {
  simTarget=GPUComputeSim
  emitterUpdateScript [
   &EmitterState(Life Cycle Mode=::NewEnumerator1, Loop Behavior=::NewEnumerator1, Loop Delay=0.6, Loop Duration=0.8, UseLoopDelay=true)
   &SpawnRate(SpawnRate=150, Use Spawn Probability=false)
  ]
  particleSpawnScript [
   &InitializeParticle(
    Lifetime=&RandomRangeFloat(Minimum=0.35, Maximum=0.64, Evaluation Type=::NewEnumerator1)
    Sprite Size=&RandomRangeVector2D(Minimum=<0.6 2.5>, Maximum=<2 4>, Evaluation Type=::NewEnumerator1)
    Sprite Size Mode=::NewEnumerator4
    UsePositionOffset=false
    Write Material Random=true
   )
   &Color(Color=&MakeLinearColorFromVectorAndFloat("Vector (RGB)"=<1 1 1>, "Float (Alpha)"=3))
   &AddVelocity(
    Rotation Coordinate Space=::World
    Velocity=&RandomRangeVector(Minimum=<-20 -20 220>, Maximum=<20 20 380>, Evaluation Type=::NewEnumerator1)
   )
   &ShapeLocation(
    Cylinder Height=8
    Cylinder Radius=&RandomRangeFloat(Minimum=2, Maximum=3, Evaluation Type=::NewEnumerator1)
    MaskSpawnGroup=false
    Override Seed=false
    Rotation Mode=::NewEnumerator2
    Shape Primitive=::NewEnumerator1
    Surface Only=true
    Use Endcaps In Surface Only Mode=false
    UseOwnerScale=true
   )
   &FixShapeVector()
   &AddVelocity(Velocity=&Multiply_VectorByFloat(Vector=&Multiply_Vector(A=$Output.FixShapeVector.ShapeVector, B=<1 1 0>), Float=25)) {module=AddVelocity001}
   &AddVectorToPosition(Vector:Vector3f=&TransformVector(Vector=<0 0 110>, Destination Space=::Simulation))
   &InheritVelocity(Inherited Velocity Amount Scale=<-0.05 -0.05 0>, Inherited Velocity Speed Limit=0, Speed Limit=false, UseSpeedThreshold=false)
  ]
  particleUpdateScript [
   &ParticleState()
   &ScaleSpriteSize(
    Scale Factor=&Vector2DFromCurve(
     Vector2Curve=DataInterface() {propertyValues {
       XCurve {keys [{interpMode=RCIM_Cubic, tangentMode=RCTM_User, time=0, value=1}, <at 1 0.4>]}
       YCurve {keys [{arriveTangent=-3.0333, interpMode=RCIM_Cubic, leaveTangent=-3.0333, tangentMode=RCTM_User, time=0, value=2}, <at 1 0.4>]}
      }}
     ScaleCurve=<1 1>
    )
    Scale Sprite Size Mode=::NewEnumerator2
   )
   &ScaleColor(
    Scale Alpha=&FloatFromCurve(FloatCurve=DataInterface() {propertyValues {Curve {keys [
         {arriveTangent=11.6133, interpMode=RCIM_Cubic, leaveTangent=11.6133, tangentMode=RCTM_User, time=0, value=0}
         <at 0.15 1 RCIM_Cubic>
         {arriveTangent=-2.6975, interpMode=RCIM_Cubic, leaveTangent=-2.6975, tangentMode=RCTM_User, time=1, value=0}
        ]}}})
    Scale RGB=&VectorFromCurve(VectorCurve=DataInterface() {propertyValues {
       XCurve {keys [<at 0.25 0.9062 RCIM_Cubic>,<at 0.85 0.333>]}
       YCurve {keys [<at 0.25 1 RCIM_Cubic>,<at 0.85 0.5274>]}
       ZCurve {keys [<at 0.25 0.9468 RCIM_Cubic>,<at 0.85 0.6>]}
      }})
    ScaleA=true
    ScaleRGB=true
   )
   &AccelerationForce(Acceleration=<0 0 -900>)
   &SolveForcesAndVelocity(Clamp Velocity=false, Limit Acceleration=false)
   &ScaleSpriteSizeBySpeed(Max Scale Factor=<10 20>, Source Velocity:Vector3f=&Multiply_Vector(A=<0 0.05 1>, B=$Particles.Velocity))
  ]
  rendererClasses [&NiagaraSpriteRendererProperties]
  renderers [{rendererClass=&NiagaraSpriteRendererProperties}]
 }
 Loop {
  emitterUpdateScript [
   &EmitterState(
    Delay First Loop Only=true
    Inactive Response=::NewEnumerator1
    Life Cycle Mode=::NewEnumerator1
    Loop Delay=1.8
    Loop Duration=1
    UseLoopDelay=true
   )
   &SpawnBurst_Instantaneous(Use Spawn Probability=false, UseLoopCountLimit=false)
  ]
  particleSpawnScript [
   &InitializeParticle(Mesh Scale Mode=::NewEnumerator4, UsePositionOffset=false, Write Material Random=true)
   &Color(Color=&MakeLinearColorFromVectorAndFloat("Vector (RGB)"=<1 1 1>, "Float (Alpha)"=1))
   &AddVectorToPosition(Vector:Vector3f=&TransformVector(Vector=<0 0 86>, Destination Space=::Simulation))
  ]
  particleUpdateScript [
   &ParticleState()
   &DynamicMaterialParameters(
    Index 0 Param 1=1
    Index 0 Param 4=1
    Module.Param0WriteEnabled=true
    Module.Param1WriteEnabled=true
    Module.Param2WriteEnabled=true
    Module.Param3WriteEnabled=true
   )
  ]
  rendererClasses [&NiagaraMeshRendererProperties]
  renderers [{rendererClass=&NiagaraMeshRendererProperties, rendererIndex=0}]
 }
 SplashArcs {
  emitterUpdateScript [
   &EmitterState(Life Cycle Mode=::NewEnumerator1, Loop Behavior=::NewEnumerator1, Loop Delay=0.9, Loop Duration=0.2, UseLoopDelay=true)
   &SpawnBurst_Instantaneous(Spawn Count=4, Spawn Time=0.05, Use Spawn Probability=false, UseLoopCountLimit=false) {module=SpawnBurst_Instantaneous001}
   &SpawnBurst_Instantaneous(Spawn Count=5, Use Spawn Probability=false, UseLoopCountLimit=false)
   &SpawnRate(SpawnRate=35, Use Spawn Probability=false)
  ]
  particleSpawnScript [
   /SetVariables(Particles.Acceleration:Vector3f=&RandomRangeVector(Minimum=<0 0 -60>, Maximum=<0 0 -50>, Evaluation Type=::NewEnumerator1)) {isSetParameters=true}
   /SetVariables(Particles.MeshRotationEuler:Vector3f=&RandomRangeVector(Minimum=<0 0.17 0>, Maximum=<0 0.2 1>, Evaluation Type=::NewEnumerator1)) {isSetParameters=true}
   /SetVariables(Particles.MeshOrientation=/CascadeToNiagaraConverter/NiagaraScripts/DynamicInputScripts/MakeQuatFromEuler(Euler:Vector3f=$Particles.MeshRotationEuler, Scale:NiagaraFloat=360)) {isSetParameters=true}
   /SetVariables(Particles.MeshRotationRateBase:Vector3f=&RandomRangeVector(Minimum=<0 0 0.05>, Maximum=<0 0 0.3>, Evaluation Type=::NewEnumerator1)) {isSetParameters=true}
   &InitializeParticle(
    Lifetime=&RandomRangeFloat(Minimum=0.45, Maximum=0.7, Evaluation Type=::NewEnumerator1)
    Mesh Scale=&VectorFromFloat(Value=&RandomRangeFloat(Minimum=0.15, Maximum=0.25, Evaluation Type=::NewEnumerator1))
    Mesh Scale Mode=::NewEnumerator4
    UsePositionOffset=false
    Write Material Random=true
   )
   &AddVectorToPosition(Vector:Vector3f=&TransformVector(Vector=<0 0 105>, Destination Space=::Simulation))
   &InheritVelocity(Inherited Velocity Amount Scale=<-0.05 -0.05 -0.05>, Inherited Velocity Speed Limit=0, Speed Limit=false, UseSpeedThreshold=false)
  ]
  particleUpdateScript [
   /SetVariables(Particles.MeshRotationRate:Vector3f=$Particles.MeshRotationRateBase) {isSetParameters=true}
   /SetVariables(Particles.MeshRotationRate:Vector3f=&Multiply_Vector(
     A=$Particles.MeshRotationRate
     B=&VectorFromCurve(VectorCurve=DataInterface() {propertyValues {
        XCurve {keys [<at 0 0 RCIM_Cubic>,<at 1 0 RCIM_Cubic>]}
        YCurve {keys [<at 0 0 RCIM_Cubic>,<at 1 0 RCIM_Cubic>]}
        ZCurve {keys [<at 0 1 RCIM_Cubic>,<at 1 0 RCIM_Cubic>]}
       }})
    )) {isSetParameters=true}
   &ParticleState()
   &ScaleMeshSize(Scale Factor=&VectorFromFloat(Value=&FloatFromCurve(FloatCurve=DataInterface() {propertyValues {Curve {keys [<at 0 0.9>,<at 1 1.1>]}}})))
   &DynamicMaterialParameters(
    Index 0 Param 1=&FloatFromCurve(FloatCurve=DataInterface() {propertyValues {Curve {keys [
         {arriveTangent=1.447, time=0, value=-2}
         {arriveTangent=2.6778, interpMode=RCIM_Cubic, leaveTangent=2.6778, tangentMode=RCTM_User, time=0.8374, value=-0.553}
         {arriveTangent=21.7295, interpMode=RCIM_Cubic, leaveTangent=21.7295, tangentMode=RCTM_User, time=1, value=1}
        ]}}})
    Index 0 Param 2=&FloatFromCurve(FloatCurve=DataInterface() {propertyValues {Curve {keys [{arriveTangent=-4, time=0, value=4}, <at 1 0>]}}})
    Index 0 Param 3=0.1
    Index 0 Param 4=5
    Module.Param0WriteEnabled=true
    Module.Param1WriteEnabled=true
    Module.Param2WriteEnabled=true
    Module.Param3WriteEnabled=true
   )
   &AccelerationForce(Acceleration=$Particles.Acceleration)
   &SolveForcesAndVelocity(Clamp Velocity=false, Limit Acceleration=false)
   /CascadeToNiagaraConverter/NiagaraScripts/ModuleScripts/UpdateMeshOrientationByEuler(RotationRate:Vector3f=$Particles.MeshRotationRate)
  ]
  rendererClasses [&NiagaraMeshRendererProperties]
  renderers [{rendererClass=&NiagaraMeshRendererProperties, rendererIndex=0}]
 }
 Start {
  emitterUpdateScript [
   &EmitterState(
    Inactive Response=::NewEnumerator1
    Life Cycle Mode=::NewEnumerator1
    Loop Behavior=::NewEnumerator1
    Loop Delay=1
    Loop Duration=1
    UseLoopDelay=true
   )
   &SpawnBurst_Instantaneous(Use Spawn Probability=false, UseLoopCountLimit=false)
  ]
  particleSpawnScript [
   &InitializeParticle(Lifetime=0.8, Mesh Scale Mode=::NewEnumerator4, UsePositionOffset=false, Write Material Random=true)
   &Color(Color=&MakeLinearColorFromVectorAndFloat("Vector (RGB)"=<1 1 1>, "Float (Alpha)"=1))
   &AddVectorToPosition(Vector:Vector3f=&TransformVector(Vector=<0 0 86>, Destination Space=::Simulation))
  ]
  particleUpdateScript [
   &ParticleState()
   &ScaleMeshSize(Scale Factor=&VectorFromCurve(VectorCurve=DataInterface() {propertyValues {
       XCurve {keys [
         {interpMode=RCIM_Cubic, tangentMode=RCTM_User, time=0, value=0.1}
         {arriveTangent=1.0373, interpMode=RCIM_Cubic, leaveTangent=1.0373, tangentMode=RCTM_User, time=0.2479, value=0.5}
         {interpMode=RCIM_Cubic, tangentMode=RCTM_User, time=1, value=1}
        ]}
       YCurve {keys [
         {interpMode=RCIM_Cubic, tangentMode=RCTM_User, time=0, value=0.1}
         {arriveTangent=1.4181, interpMode=RCIM_Cubic, leaveTangent=1.4181, tangentMode=RCTM_User, time=0.2479, value=0.5}
         {arriveTangent=-0.136, interpMode=RCIM_Cubic, leaveTangent=-0.136, tangentMode=RCTM_User, time=1, value=1}
        ]}
       ZCurve {keys [
         {arriveTangent=2.902, interpMode=RCIM_Cubic, leaveTangent=2.902, tangentMode=RCTM_User, time=0, value=0.8}
         {arriveTangent=-0.3784, interpMode=RCIM_Cubic, leaveTangent=-0.3784, tangentMode=RCTM_User, time=0.2479, value=1.1069}
         {interpMode=RCIM_Cubic, tangentMode=RCTM_User, time=1, value=1}
        ]}
      }}))
   &DynamicMaterialParameters(
    Index 0 Param 1=&FloatFromCurve(FloatCurve=DataInterface() {propertyValues {Curve {keys [<at 0 0 RCIM_Cubic>,<at 1 1>]}}})
    Index 0 Param 2=&FloatFromCurve(FloatCurve=DataInterface() {propertyValues {Curve {keys [<at 0 4 RCIM_Cubic>,<at 1 0>]}}})
    Index 0 Param 3=&FloatFromCurve(FloatCurve=DataInterface() {propertyValues {Curve {keys [{arriveTangent=2.1578, interpMode=RCIM_Cubic, leaveTangent=2.1578, tangentMode=RCTM_User, time=0, value=-1}, <at 1 0>]}}})
    Index 0 Param 4=1
    Module.Param0WriteEnabled=true
    Module.Param1WriteEnabled=true
    Module.Param2WriteEnabled=true
    Module.Param3WriteEnabled=true
   )
  ]
  rendererClasses [&NiagaraMeshRendererProperties]
  renderers [{rendererClass=&NiagaraMeshRendererProperties, rendererIndex=0}]
 }
 WashInner {
  emitterUpdateScript [
   &EmitterState(Life Cycle Mode=::NewEnumerator1, Loop Behavior=::NewEnumerator1, Loop Delay=0.8, Loop Duration=0.5, UseLoopDelay=true)
   &SpawnBurst_Instantaneous(Spawn Time=0.07, Use Spawn Probability=false, UseLoopCountLimit=false) {module=SpawnBurst_Instantaneous001}
   &SpawnBurst_Instantaneous(Use Spawn Probability=false, UseLoopCountLimit=false)
  ]
  particleSpawnScript [
   /SetVariables(Particles.MeshRotationRateBase:Vector3f=<0 0 0.25>) {isSetParameters=true}
   /SetVariables(Particles.MeshRotationEuler:Vector3f=&RandomRangeVector(Minimum=<0 0 0>, Maximum=<0 0 1>, Evaluation Type=::NewEnumerator1)) {isSetParameters=true}
   /SetVariables(Particles.MeshOrientation=&MakeQuatFromEuler(Euler:Vector3f=$Particles.MeshRotationEuler, Scale:NiagaraFloat=360)) {isSetParameters=true}
   &InitializeParticle(Lifetime=0.8, Mesh Scale=<0.8 0.8 0.8>, Mesh Scale Mode=::NewEnumerator4, UsePositionOffset=false, Write Material Random=true)
   &Color(Color=&MakeLinearColorFromVectorAndFloat("Vector (RGB)"=<1 1 1>, "Float (Alpha)"=1))
  ]
  particleUpdateScript [
   /SetVariables(Particles.MeshRotationRate:Vector3f=$Particles.MeshRotationRateBase) {isSetParameters=true}
   /SetVariables(Particles.MeshRotationRate:Vector3f=&Multiply_Vector(
     A=$Particles.MeshRotationRate
     B=&VectorFromCurve(VectorCurve=DataInterface() {propertyValues {
        XCurve {keys [<at 0 0 RCIM_Cubic>,<at 1 0 RCIM_Cubic>]}
        YCurve {keys [<at 0 0 RCIM_Cubic>,<at 1 0 RCIM_Cubic>]}
        ZCurve {keys [<at 0 1 RCIM_Cubic>,<at 1 0 RCIM_Cubic>]}
       }})
    )) {isSetParameters=true}
   /SetVariables(Particles.Position=&TransformPosition(
     Position=&ConvertVectorToPosition(
      Input Position=&VectorFromCurve(VectorCurve=DataInterface() {propertyValues {
         XCurve {keys [
           {interpMode=RCIM_Cubic, tangentMode=RCTM_User, time=0, value=0}
           {interpMode=RCIM_Cubic, tangentMode=RCTM_User, time=0.25, value=0}
           {interpMode=RCIM_Cubic, tangentMode=RCTM_User, time=1, value=0}
          ]}
         YCurve {keys [
           {interpMode=RCIM_Cubic, tangentMode=RCTM_User, time=0, value=0}
           {interpMode=RCIM_Cubic, tangentMode=RCTM_User, time=0.25, value=0}
           {interpMode=RCIM_Cubic, tangentMode=RCTM_User, time=1, value=0}
          ]}
         ZCurve {keys [
           {arriveTangent=253.1079, interpMode=RCIM_Cubic, leaveTangent=253.1079, tangentMode=RCTM_User, time=0, value=70}
           {arriveTangent=-11.8636, interpMode=RCIM_Cubic, leaveTangent=-11.8636, tangentMode=RCTM_User, time=0.25, value=100}
           {arriveTangent=-20.1484, interpMode=RCIM_Cubic, leaveTangent=-20.1484, tangentMode=RCTM_User, time=1, value=70}
          ]}
        }})
      Conversion Operation:ENiagara_LWCConvertVecToPos=::NewEnumerator0
     )
     Source Space=::Local
     Destination Space=::Simulation
    )) {isSetParameters=true}
   &ParticleState()
   &ScaleMeshSize(Scale Factor=&VectorFromCurve(VectorCurve=DataInterface() {propertyValues {
       XCurve {keys [
         {interpMode=RCIM_Cubic, tangentMode=RCTM_User, time=0, value=0.4}
         {arriveTangent=0.5, interpMode=RCIM_Cubic, leaveTangent=0.5, time=0.25, value=0.7}
         {interpMode=RCIM_Cubic, tangentMode=RCTM_User, time=1, value=0.9}
        ]}
       YCurve {keys [
         {interpMode=RCIM_Cubic, tangentMode=RCTM_User, time=0, value=0.4}
         {arriveTangent=0.5, interpMode=RCIM_Cubic, leaveTangent=0.5, time=0.25, value=0.7}
         {arriveTangent=-0.136, interpMode=RCIM_Cubic, leaveTangent=-0.136, tangentMode=RCTM_User, time=1, value=0.9}
        ]}
       ZCurve {keys [
         {arriveTangent=2.5944, interpMode=RCIM_Cubic, leaveTangent=2.5944, tangentMode=RCTM_User, time=0, value=0.6}
         {arriveTangent=0.4, interpMode=RCIM_Cubic, leaveTangent=0.4, time=0.25, value=0.9591}
         {interpMode=RCIM_Cubic, tangentMode=RCTM_User, time=1, value=1}
        ]}
      }}))
   &DynamicMaterialParameters(
    Index 0 Param 1=&FloatFromCurve(FloatCurve=DataInterface() {propertyValues {Curve {keys [<at 0 0 RCIM_Cubic>,<at 1 0.5>]}}})
    Index 0 Param 2=1
    Index 0 Param 3=&FloatFromCurve(FloatCurve=DataInterface() {propertyValues {Curve {keys [<at 0 1 RCIM_Cubic>,<at 1 1.5>]}}})
    Index 0 Param 4=1
    Module.Param0WriteEnabled=true
    Module.Param1WriteEnabled=true
    Module.Param2WriteEnabled=true
    Module.Param3WriteEnabled=true
   )
   &UpdateMeshOrientationByEuler(RotationRate:Vector3f=$Particles.MeshRotationRate)
   &ScaleColor(
    Scale Alpha=&FloatFromCurve(FloatCurve=DataInterface() {propertyValues {Curve {keys [
         {arriveTangent=11.9078, interpMode=RCIM_Cubic, leaveTangent=11.9078, tangentMode=RCTM_User, time=0, value=0}
         {interpMode=RCIM_Cubic, tangentMode=RCTM_User, time=0.15, value=1}
         {arriveTangent=-1.1765, interpMode=RCIM_Cubic, leaveTangent=-1.1765, time=0.7418, value=0.7744}
         <at 1 0>
        ]}}})
    ScaleA=true
    ScaleRGB=true
   )
   &CameraOffset(Camera Offset Amount=15)
  ]
  rendererClasses [&NiagaraMeshRendererProperties]
  renderers [{rendererClass=&NiagaraMeshRendererProperties, rendererIndex=0}]
 }
 WashOuter {
  emitterUpdateScript [
   &EmitterState(Life Cycle Mode=::NewEnumerator1, Loop Behavior=::NewEnumerator1, Loop Delay=0.8, Loop Duration=0.5, UseLoopDelay=true)
   &SpawnBurst_Instantaneous(Spawn Time=0.12, Use Spawn Probability=false, UseLoopCountLimit=false) {module=SpawnBurst_Instantaneous001}
   &SpawnBurst_Instantaneous(Use Spawn Probability=false, UseLoopCountLimit=false)
  ]
  particleSpawnScript [
   /SetVariables(Particles.Acceleration:Vector3f=&RandomRangeVector(Minimum=<0 0 -200>, Maximum=<0 0 -150>, Evaluation Type=::NewEnumerator1)) {isSetParameters=true}
   /SetVariables(Particles.MeshRotationRateBase:Vector3f=&RandomRangeVector(Minimum=<0 0 0.35>, Maximum=<0 0 0.6>, Evaluation Type=::NewEnumerator1)) {isSetParameters=true}
   /SetVariables(Particles.MeshRotationEuler:Vector3f=&RandomRangeVector(Minimum=<0 0 0>, Maximum=<0 0 1>, Evaluation Type=::NewEnumerator1)) {isSetParameters=true}
   /SetVariables(Particles.MeshOrientation=&MakeQuatFromEuler(Euler:Vector3f=$Particles.MeshRotationEuler, Scale:NiagaraFloat=360)) {isSetParameters=true}
   &InitializeParticle(
    Lifetime=&RandomRangeFloat(Minimum=0.5, Maximum=0.65, Evaluation Type=::NewEnumerator1)
    Mesh Scale Mode=::NewEnumerator4
    UsePositionOffset=false
    Write Material Random=true
   )
   &Color(Color=&MakeLinearColorFromVectorAndFloat("Vector (RGB)"=<1 1 1>, "Float (Alpha)"=1))
   &AddVectorToPosition(Vector:Vector3f=&TransformVector(Vector=<0 0 90>, Destination Space=::Simulation))
  ]
  particleUpdateScript [
   /SetVariables(Particles.MeshRotationRate:Vector3f=$Particles.MeshRotationRateBase) {isSetParameters=true}
   /SetVariables(Particles.MeshRotationRate:Vector3f=&Multiply_Vector(
     A=$Particles.MeshRotationRate
     B=&VectorFromCurve(VectorCurve=DataInterface() {propertyValues {
        XCurve {keys [<at 0 0 RCIM_Cubic>,<at 1 0 RCIM_Cubic>]}
        YCurve {keys [<at 0 0 RCIM_Cubic>,<at 1 0 RCIM_Cubic>]}
        ZCurve {keys [<at 0 1 RCIM_Cubic>,<at 1 0 RCIM_Cubic>]}
       }})
    )) {isSetParameters=true}
   &ParticleState()
   &ScaleMeshSize(Scale Factor=&VectorFromCurve(VectorCurve=DataInterface() {propertyValues {
       XCurve {keys [
         {interpMode=RCIM_Cubic, tangentMode=RCTM_User, time=0, value=0.6}
         {arriveTangent=1.4262, interpMode=RCIM_Cubic, leaveTangent=1.4262, tangentMode=RCTM_User, time=0.2825, value=0.8}
         {arriveTangent=-0.8754, interpMode=RCIM_Cubic, leaveTangent=-0.8754, tangentMode=RCTM_User, time=1, value=0.8}
        ]}
       YCurve {keys [
         {interpMode=RCIM_Cubic, tangentMode=RCTM_User, time=0, value=0.6}
         {arriveTangent=1.3075, interpMode=RCIM_Cubic, leaveTangent=1.3075, tangentMode=RCTM_User, time=0.2825, value=0.8}
         {arriveTangent=-0.8381, interpMode=RCIM_Cubic, leaveTangent=-0.8381, tangentMode=RCTM_User, time=1, value=0.8}
        ]}
       ZCurve {keys [
         {arriveTangent=4.7272, interpMode=RCIM_Cubic, leaveTangent=4.7272, tangentMode=RCTM_User, time=0, value=0.4}
         {arriveTangent=0.9946, interpMode=RCIM_Cubic, leaveTangent=0.9946, tangentMode=RCTM_User, time=0.2825, value=0.986}
         {arriveTangent=1.243, interpMode=RCIM_Cubic, leaveTangent=1.243, tangentMode=RCTM_User, time=1, value=2}
        ]}
      }}))
   &DynamicMaterialParameters(
    Index 0 Param 1=&FloatFromCurve(FloatCurve=DataInterface() {propertyValues {Curve {keys [<at 0 0 RCIM_Cubic>,<at 1 1>]}}})
    Index 0 Param 2=1
    Index 0 Param 3=&FloatFromCurve(FloatCurve=DataInterface() {propertyValues {Curve {keys [<at 0 0.6 RCIM_Cubic>,<at 1 1.25>]}}})
    Index 0 Param 4=1
    Module.Param0WriteEnabled=true
    Module.Param1WriteEnabled=true
    Module.Param2WriteEnabled=true
    Module.Param3WriteEnabled=true
   )
   &AccelerationForce(Acceleration=$Particles.Acceleration)
   &SolveForcesAndVelocity(Clamp Velocity=false, Limit Acceleration=false)
   &UpdateMeshOrientationByEuler(RotationRate:Vector3f=$Particles.MeshRotationRate)
   &ScaleColor(
    Scale Alpha=&FloatFromCurve(FloatCurve=DataInterface() {propertyValues {Curve {keys [
         {arriveTangent=16.6129, interpMode=RCIM_Cubic, leaveTangent=16.6129, tangentMode=RCTM_User, time=0, value=0}
         {arriveTangent=-1.6181, interpMode=RCIM_Cubic, leaveTangent=-1.6181, tangentMode=RCTM_User, time=0.15, value=1}
         <at 1 0>
        ]}}})
    ScaleA=true
    ScaleRGB=true
   )
   &CameraOffset(Camera Offset Amount=15)
  ]
  rendererClasses [&NiagaraMeshRendererProperties]
  renderers [{rendererClass=&NiagaraMeshRendererProperties, rendererIndex=0}]
 }
}
summary {
 systemName=NS_Umbrella_S11
 emitters [
  {emitterName=Start, rendererClasses [&NiagaraMeshRendererProperties]}
  {emitterName=Loop, rendererClasses [&NiagaraMeshRendererProperties]}
  {emitterName=WashInner, rendererClasses [&NiagaraMeshRendererProperties]}
  {emitterName=WashOuter, rendererClasses [&NiagaraMeshRendererProperties]}
  {emitterName=SplashArcs, rendererClasses [&NiagaraMeshRendererProperties]}
  {emitterName=DripsStart, simTarget=GPUComputeSim, rendererClasses [&NiagaraSpriteRendererProperties]}
  {emitterName=DripsLoop, simTarget=GPUComputeSim, rendererClasses [&NiagaraSpriteRendererProperties]}
  {emitterName=DripsStartTop, simTarget=GPUComputeSim, rendererClasses [&NiagaraSpriteRendererProperties]}
  {emitterName=DripsLoopTop, simTarget=GPUComputeSim, rendererClasses [&NiagaraSpriteRendererProperties]}
 ]
 userVariables []
}
