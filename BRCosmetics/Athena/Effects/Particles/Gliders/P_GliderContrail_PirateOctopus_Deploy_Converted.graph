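# Niagara deploy effect for the Pirate Octopus glider, converted from Cascade: a golden burst of tumbling, falling coin meshes.
# Coins combine a burst with a decaying spawn rate, joined by GPU glints and curl-noise sparks.
# Flare, corona, shield decal, smoke and expanding glow mesh layers add the flash, plus a brief distance-culled light.
# Every emitter samples the User.SK_Mesh skeletal mesh for spawn location; curves drive fades and material parameters.
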NiagaraSystem /BRCosmetics/Athena/Effects/Particles/Gliders/P_GliderContrail_PirateOctopus_Deploy_Converted
dependencies {
 usedDataInterfaces [&NiagaraDataInterfaceCurve, &NiagaraDataInterfaceVectorCurve]
 usedDynamicInputs []
 usedModules [&SystemState]
}
emitters {
 Coins {
  emitterUpdateScript [
   &EmitterState(Life Cycle Mode=::NewEnumerator1, Loop Behavior=::NewEnumerator1, Loop Delay=0.13, Loop Duration=0.2, UseLoopDelay=true)
   &SpawnBurst_Instantaneous(Spawn Count=5, Spawn Time=0.1, Use Spawn Probability=false, UseLoopCountLimit=false)
   &SpawnRate(
    SpawnRate=&Multiply_Float(
     A=125
     B=&FloatFromCurve(FloatCurve=DataInterface() {propertyValues {Curve {keys [<at 0 1>,<at 1 0>]}}}, CurveIndex=$Emitter.NormalizedLoopAge)
    )
    Use Spawn Probability=false
   )
  ]
  particleSpawnScript [
   /SetVariables(Particles.SpawnRotationVector=&RandomRangeVector(Minimum=<1 1 1>, Maximum=<3 3 3>)) {isSetParameters=true}
   /SetVariables(Particles.Acceleration:Vector3f=&RandomRangeVector(Minimum=<0 0 -1200>, Maximum=<0 0 -1600>)) {isSetParameters=true}
   /SetVariables(Particles.MeshOrientation=&MakeQuatFromAxisAngle(Axis=&RandomRangeVector(Minimum=<0 0 0>, Maximum=<1 1 1>), AngleInDegrees=90)) {isSetParameters=true}
   &InitializeParticle(
    Lifetime=&RandomRangeFloat(Minimum=0.5, Maximum=1)
    Mesh Scale=/Niagara/DynamicInputs/CascadeConversion/SplitVector(
     Double Component Vector Channels:ECascadeNiagaraTwoVectorChannels=::NewEnumerator0
     Double Component:NiagaraFloat=&RandomRangeFloat(Minimum=0.04, Maximum=0.05)
     Single Component:NiagaraFloat=&RandomRangeFloat(Minimum=0.06, Maximum=0.08)
    )
    Mesh Scale Mode=::NewEnumerator4
    Position=&RandomRangeVector(Minimum=<-5 0 5>, Maximum=<15 0 10>)
    Position Mode=::NewEnumerator1
    UsePositionOffset=false
    Write Material Random=true
   )
   &SkeletalMeshLocation(MaskSpawnGroup=false, Mesh Sampling Type=::NewEnumerator5, Skeletal Mesh=$User.SK_Mesh, Socket Sampling Mode=::NewEnumerator5)
   &Color(Color=&MakeLinearColorFromVectorAndFloat("Vector (RGB)"=<1 0.8784 0.4941>, "Float (Alpha)"=&ClampFloat(Float=1)))
   &AddVelocity(Velocity=&Multiply_VectorByFloat(Vector=&UniformRangedVector(Minimum=<-400 -400 -400>, Maximum=<400 400 400>), Float=0.5)) {module=AddVelocity002}
   &ApplyInitialForces()
   &CameraOffset(Camera Offset Amount=25)
  ]
  particleUpdateScript [
   &ParticleState()
   &ScaleMeshSize(Scale Factor=<1.2 1.2 1.2>)
   &UpdateMeshOrientation(Rotation Vector=$Particles.SpawnRotationVector)
   &AccelerationForce(Acceleration=$Particles.Acceleration)
   &SolveForcesAndVelocity(Clamp Velocity=false, Limit Acceleration=false)
  ]
  rendererClasses [&NiagaraMeshRendererProperties]
  renderers [{rendererClass=&NiagaraMeshRendererProperties, rendererIndex=0}]
 }
 ColdSmoke {
  emitterUpdateScript [
   &EmitterState(Life Cycle Mode=::NewEnumerator1, Loop Behavior=::NewEnumerator1, Loop Duration=2, UseLoopDelay=true)
   &SpawnBurst_Instantaneous(Spawn Count=3, Use Spawn Probability=false, UseLoopCountLimit=false)
  ]
  particleSpawnScript [
   &InitializeParticle(
    Lifetime=&RandomRangeFloat(Minimum=0.8, Maximum=1.5)
    Position=<0 50 0>
    Position Mode=::NewEnumerator1
    Sprite Rotation Angle=&RandomRangeFloat(Minimum=0, Maximum=1)
    Sprite Rotation Mode=::NewEnumerator1
    Sprite Size=&Vector2DFromFloat(Value=&RandomRangeFloat(Minimum=100, Maximum=150))
    Sprite Size Mode=::NewEnumerator4
    UsePositionOffset=false
    Write Material Random=true
   )
   &Color(Color=&MakeLinearColorFromVectorAndFloat(
     "Vector (RGB)"=&RandomRangeVector(Minimum=<0.7 0.7 0.7>, Maximum=<1 1 1>)
     "Float (Alpha)"=&ClampFloat(Float=0.1)
    ))
   &SkeletalMeshLocation(MaskSpawnGroup=false, Mesh Sampling Type=::NewEnumerator5, Skeletal Mesh=$User.SK_Mesh, Socket Sampling Mode=::NewEnumerator5)
  ]
  particleUpdateScript [
   &ParticleState()
   &ScaleSpriteSize()
   &ScaleColor(
    Scale Alpha=&FloatFromCurve(FloatCurve=DataInterface() {propertyValues {Curve {keys [
         {arriveTangent=45.795, interpMode=RCIM_Cubic, time=0, value=0}
         {arriveTangent=1.0985, interpMode=RCIM_Cubic, leaveTangent=1.0985, time=0.0433, value=0.8697}
         {arriveTangent=-0.909, interpMode=RCIM_Cubic, leaveTangent=-0.909, time=0.2126, value=0.2336}
         {interpMode=RCIM_Cubic, leaveTangent=-0.0561, time=1, value=0}
        ]}}})
    Scale RGB=&VectorFromCurve(VectorCurve=DataInterface() {propertyValues {XCurve {keys [<at 0 5>,<at 1 1>]}, YCurve {keys [<at 0 2.0294>,<at 1 0.5034>]}, ZCurve {keys [<at 0 0>,<at 1 0.38>]}}})
    ScaleA=true
    ScaleRGB=true
   )
   /Niagara/Modules/Update/SubUV/SubUVAnimation(Number Of Frames=63)
   &CameraOffset(Camera Offset Amount=150)
   &DynamicMaterialParameters(
    Index 0 Param 1=2
    Index 0 Param 2=1
    Index 0 Param 3=1
    Index 0 Param 4=1
    Module.Param0WriteEnabled=true
    Module.Param1WriteEnabled=true
    Module.Param2WriteEnabled=true
    Module.Param3WriteEnabled=true
   )
  ]
  rendererClasses [&NiagaraSpriteRendererProperties]
  renderers [{rendererClass=&NiagaraSpriteRendererProperties}]
 }
 Corona {
  emitterUpdateScript [
   &EmitterState(Life Cycle Mode=::NewEnumerator1, Loop Behavior=::NewEnumerator1, Loop Duration=1, UseLoopDelay=true)
   &SpawnBurst_Instantaneous(Use Spawn Probability=false, UseLoopCountLimit=false)
  ]
  particleSpawnScript [
   &InitializeParticle(
    Lifetime=0.35
    Position=<0 50 5>
    Position Mode=::NewEnumerator1
    Sprite Size=<250 250>
    Sprite Size Mode=::NewEnumerator4
    UsePositionOffset=false
    Write Material Random=true
   )
   &Color(Color=&MakeLinearColorFromVectorAndFloat("Vector (RGB)"=<25 9.0396 0>, "Float (Alpha)"=&ClampFloat(Float=2)))
   &DynamicMaterialParameters(
    Index 0 Param 3=&RandomRangeFloat(Minimum=0.25, Maximum=0.35)
    Module.Param0WriteEnabled=false
    Module.Param1WriteEnabled=false
    Module.Param2WriteEnabled=true
    Module.Param3WriteEnabled=false
   ) {module=DynamicMaterialParameters001}
   &SkeletalMeshLocation(MaskSpawnGroup=false, Mesh Sampling Type=::NewEnumerator5, Skeletal Mesh=$User.SK_Mesh, Socket Sampling Mode=::NewEnumerator5)
  ]
  particleUpdateScript [
   &ParticleState()
   &ScaleSpriteSize()
   &CameraOffset(Camera Offset Amount=25)
   &DynamicMaterialParameters(
    Index 0 Param 1=&FloatFromCurve(FloatCurve=DataInterface() {propertyValues {Curve {keys [{arriveTangent=-1.2514, interpMode=RCIM_Cubic, time=0, value=1.6}, <at 1 0.5>]}}})
    Index 0 Param 2=1
    Index 0 Param 4=1
    Module.Param0WriteEnabled=true
    Module.Param1WriteEnabled=true
    Module.Param2WriteEnabled=false
    Module.Param3WriteEnabled=true
   )
   &ScaleColor(
    Scale Alpha=&FloatFromCurve(FloatCurve=DataInterface() {propertyValues {Curve {keys [
         <at -0.0045 0.0182 RCIM_Cubic>
         {arriveTangent=3.8563, interpMode=RCIM_Cubic, leaveTangent=3.8563, time=0.0333, value=1}
         {arriveTangent=-2.2133, interpMode=RCIM_Cubic, leaveTangent=-2.2133, time=0.1601, value=0.6529}
         {arriveTangent=-0.7773, interpMode=RCIM_Cubic, leaveTangent=-0.7773, time=0.3612, value=0.2742}
         <at 1 0>
        ]}}})
    Scale RGB=&VectorFromCurve(VectorCurve=DataInterface() {propertyValues {
       XCurve {keys [<at 0 1 RCIM_Cubic>,<at 1 0.3728>]}
       YCurve {keys [<at 0 0.6851 RCIM_Cubic>,<at 1 1>]}
       ZCurve {keys [<at 0 0.8647 RCIM_Cubic>,<at 1 0.6644>]}
      }})
    ScaleA=true
    ScaleRGB=true
   )
  ]
  rendererClasses [&NiagaraSpriteRendererProperties]
  renderers [{rendererClass=&NiagaraSpriteRendererProperties}]
 }
 FlashFlare {
  emitterUpdateScript [
   &EmitterState(Life Cycle Mode=::NewEnumerator1, Loop Behavior=::NewEnumerator1, Loop Duration=1, UseLoopDelay=true)
   &SpawnBurst_Instantaneous(Use Spawn Probability=false, UseLoopCountLimit=false)
  ]
  particleSpawnScript [
   &InitializeParticle(
    Lifetime=0.25
    Position=<5 50 10>
    Position Mode=::NewEnumerator1
    Sprite Size=<250 250>
    Sprite Size Mode=::NewEnumerator4
    UsePositionOffset=false
    Write Material Random=true
   )
   &Color(Color=&MakeLinearColorFromVectorAndFloat("Vector (RGB)"=<1.5 0.6763 0>, "Float (Alpha)"=&ClampFloat(Float=0.25)))
   &DynamicMaterialParameters(
    Index 0 Param 1=1
    Module.Param0WriteEnabled=true
    Module.Param1WriteEnabled=false
    Module.Param2WriteEnabled=false
    Module.Param3WriteEnabled=false
   )
   &SkeletalMeshLocation(MaskSpawnGroup=false, Mesh Sampling Type=::NewEnumerator5, Skeletal Mesh=$User.SK_Mesh, Socket Sampling Mode=::NewEnumerator5)
  ]
  particleUpdateScript [
   &ParticleState()
   &ScaleSpriteSize(Uniform Curve Scale=2)
   &ScaleColor(
    Scale Alpha=&FloatFromCurve(FloatCurve=DataInterface() {propertyValues {Curve {keys [<at 0 1 RCIM_Cubic>,<at 1 0>]}}})
    Scale RGB=<4 4 4>
    ScaleA=true
    ScaleRGB=true
   )
   &CameraOffset(Camera Offset Amount=50)
   &DynamicMaterialParameters(
    Index 0 Param 2=1
    Index 0 Param 3=1
    Index 0 Param 4=1
    Module.Param0WriteEnabled=false
    Module.Param1WriteEnabled=true
    Module.Param2WriteEnabled=true
    Module.Param3WriteEnabled=true
   ) {module=DynamicMaterialParameters001}
  ]
  rendererClasses [&NiagaraSpriteRendererProperties]
  renderers [{rendererClass=&NiagaraSpriteRendererProperties}]
 }
 Glints {
  simTarget=GPUComputeSim
  emitterUpdateScript [
   &EmitterState(Life Cycle Mode=::NewEnumerator1, Loop Behavior=::NewEnumerator1, Loop Delay=0.06, Loop Duration=2, UseLoopDelay=false)
   &SpawnBurst_Instantaneous(Spawn Count=25, Use Spawn Probability=false, UseLoopCountLimit=false)
  ]
  particleSpawnScript [
   &InitializeParticle(
    Lifetime=&RandomRangeFloat(Minimum=0.25, Maximum=0.8)
    Position=&RandomRangeVector(Minimum=<-5 25 5>, Maximum=<15 50 10>)
    Position Mode=::NewEnumerator1
    Sprite Size=&RandomRangeVector2D(Minimum=<25 5>, Maximum=<50 30>)
    Sprite Size Mode=::NewEnumerator4
    UsePositionOffset=false
    Write Material Random=true
   )
   &SkeletalMeshLocation(MaskSpawnGroup=false, Mesh Sampling Type=::NewEnumerator5, Skeletal Mesh=$User.SK_Mesh, Socket Sampling Mode=::NewEnumerator5)
   &Color(Color=&MakeLinearColorFromVectorAndFloat("Vector (RGB)"=<1 1 1>, "Float (Alpha)"=&ClampFloat(Float=1)))
   &SphereLocation(
    Hemisphere X=true
    MaskSpawnGroup=false
    Override Seed=false
    Sphere Radius=5
    Surface Only=false
    UseNonUniformScale=false
    UseOffset=false
    UseSphereOrientation=false
   ) {module=SphereLocation001}
   &AddVelocity(Velocity=&Multiply_VectorByFloat(Vector=&UniformRangedVector(Minimum=<-400 -400 -400>, Maximum=<400 400 400>), Float=0.15)) {module=AddVelocity002}
   &ApplyInitialForces()
   &CameraOffset(Camera Offset Amount=50)
  ]
  particleUpdateScript [
   &ParticleState()
   &ScaleSpriteSize()
   &AccelerationForce(Acceleration=<0 0 -250>)
   &SolveForcesAndVelocity(Clamp Velocity=false, Limit Acceleration=false)
   &ScaleColor(
    Scale RGB=&VectorFromCurve(VectorCurve=DataInterface() {propertyValues {XCurve {keys [<at 0 1>,<at 0.5 1>]}, YCurve {keys [<at 0 0.4452>,<at 0.5 0.7033>]}, ZCurve {keys [<at 0 0.1082>,<at 0.5 0>]}}})
    ScaleA=true
    ScaleRGB=true
   )
  ]
  rendererClasses [&NiagaraSpriteRendererProperties]
  renderers [{rendererClass=&NiagaraSpriteRendererProperties}]
 }
 GlowStart {
  emitterUpdateScript [
   &EmitterState(Life Cycle Mode=::NewEnumerator1, Loop Behavior=::NewEnumerator1, Loop Duration=0.25, UseLoopDelay=true)
   &SpawnBurst_Instantaneous(Use Spawn Probability=false, UseLoopCountLimit=false)
  ]
  particleSpawnScript [
   &InitializeParticle(
    Lifetime=&RandomRangeFloat(Minimum=0.35, Maximum=0.5)
    Mesh Scale=<3 3 3>
    Mesh Scale Mode=::NewEnumerator4
    Position=&RandomRangeVector(Minimum=<-5 50 2>, Maximum=<15 50 2>)
    Position Mode=::NewEnumerator1
    UsePositionOffset=false
    Write Material Random=true
   )
   &Color(Color=&MakeLinearColorFromVectorAndFloat("Vector (RGB)"=<1 0.455 0>, "Float (Alpha)"=&ClampFloat(Float=0.1)))
   &SkeletalMeshLocation(MaskSpawnGroup=false, Mesh Sampling Type=::NewEnumerator5, Skeletal Mesh=$User.SK_Mesh, Socket Sampling Mode=::NewEnumerator5)
  ]
  particleUpdateScript [
   &ParticleState()
   &ScaleMeshSize(Scale Factor=&VectorFromCurve(VectorCurve=DataInterface() {propertyValues {XCurve {keys [<at 0 1>,<at 1 2>]}, YCurve {keys [<at 0 1>,<at 1 2>]}, ZCurve {keys [<at 0 1>,<at 1 2>]}}}))
   &DynamicMaterialParameters(
    Index 0 Param 1=&FloatFromCurve(FloatCurve=DataInterface() {propertyValues {Curve {keys [<at 0 2>,<at 1 4>]}}})
    Index 0 Param 2=25
    Index 0 Param 3=&FloatFromCurve(FloatCurve=DataInterface() {propertyValues {Curve {keys [<at 0 25>,<at 1 1>]}}})
    Index 0 Param 4=&FloatFromCurve(FloatCurve=DataInterface() {propertyValues {Curve {keys [<at 0 1.5>,<at 1 1>]}}})
    Module.Param0WriteEnabled=true
    Module.Param1WriteEnabled=true
    Module.Param2WriteEnabled=true
    Module.Param3WriteEnabled=true
   )
   &CameraOffset(Camera Offset Amount=25)
   &ScaleColor(
    Scale Alpha=&FloatFromCurve(FloatCurve=DataInterface() {propertyValues {Curve {keys [<at 0.5 1>,<at 1 0>]}}})
    Scale RGB=&VectorFromCurve(VectorCurve=DataInterface() {propertyValues {XCurve {keys [<at 0 1>,<at 1 1>,<at 1 0>]}, YCurve {keys [<at 0 0.804>,<at 1 0.339>,<at 1 0>]}, ZCurve {keys [<at 0 0>,<at 1 0>,<at 1 0>]}}})
    ScaleA=true
    ScaleRGB=true
   )
  ]
  rendererClasses [&NiagaraMeshRendererProperties]
  renderers [{rendererClass=&NiagaraMeshRendererProperties, rendererIndex=0}]
 }
 HQ_Light {
  emitterUpdateScript [
   &EmitterState(
    Enable Distance Culling=true
    Inactive Response=::NewEnumerator1
    Life Cycle Mode=::NewEnumerator1
    Loop Behavior=::NewEnumerator1
    Loop Duration=1
    MaxDistance=1500
    Module.bUseMaxDistance=true
    Module.bUseMinDistance=false
    Scalability Mode=::NewEnumerator1
    UseLoopDelay=false
   )
   &SpawnBurst_Instantaneous(Spawn Count=1, Use Spawn Probability=false)
  ]
  particleSpawnScript [
   &InitializeParticle(Color Mode=::NewEnumerator0, Lifetime=0.5, Position Mode=::NewEnumerator1, UsePositionOffset=false, Write Material Random=true)
   &SkeletalMeshLocation(MaskSpawnGroup=false, Mesh Sampling Type=::NewEnumerator5, Skeletal Mesh=$User.SK_Mesh, Socket Sampling Mode=::NewEnumerator5)
  ]
  particleUpdateScript [
   &ParticleState()
   &FN_Light_Attributes(
    Color=<rgba 15 6.5 0 0.7>
    Create Color Attribute=true
    Create Exponent Attribute=true
    Create Position Attribute=true
    Create Radius Attribute=true
    Enable Visibility Tag=true
    Exponent=7.5
    Position=<7.5 60 15>
    Radius=50
    Scale Color=&FloatFromCurve(
     FloatCurve=DataInterface() {propertyValues {Curve {keys [
         {arriveTangent=117.3003, interpMode=RCIM_Cubic, leaveTangent=117.3003, tangentMode=RCTM_User, time=0, value=0}
         {arriveTangent=1.3594, interpMode=RCIM_Cubic, leaveTangent=1.3594, tangentMode=RCTM_User, time=0.0286, value=1}
         {arriveTangent=-1.0294, interpMode=RCIM_Cubic, leaveTangent=-1.0294, time=0.95, value=0.9805}
         {arriveTangent=-65.8727, interpMode=RCIM_Cubic, leaveTangent=-65.8727, tangentMode=RCTM_User, time=1, value=0}
        ]}}}
     Scale Curve=3
    )
    VisibilityTag=HlslExpression() {hlslExpression="Particles.UniqueID && 1"}
   )
  ]
  rendererClasses [&NiagaraLightRendererProperties]
  renderers [{rendererClass=&NiagaraLightRendererProperties, rendererIndex=0}]
 }
 ShieldDecals {
  emitterUpdateScript [
   &EmitterState(Life Cycle Mode=::NewEnumerator1, Loop Behavior=::NewEnumerator1, Loop Duration=1, UseLoopDelay=true)
   &SpawnBurst_Instantaneous(Use Spawn Probability=false, UseLoopCountLimit=false)
  ]
  particleSpawnScript [
   &InitializeParticle(
    Lifetime=0.5
    Position=&RandomRangeVector(Minimum=<-5 50 2>, Maximum=<15 50 2>)
    Position Mode=::NewEnumerator1
    Sprite Rotation Angle=&RandomRangeFloat(Minimum=0, Maximum=1)
    Sprite Rotation Mode=::NewEnumerator1
    Sprite Size=<150 150>
    Sprite Size Mode=::NewEnumerator4
    UsePositionOffset=false
    Write Material Random=true
   )
   &Color(Color=&MakeLinearColorFromVectorAndFloat(
     "Vector (RGB)"=&RandomRangeVector(Minimum=<25 17.0366 0>, Maximum=<25 5.5411 0>)
     "Float (Alpha)"=&ClampFloat(Float=0.02)
    ))
   &SkeletalMeshLocation(MaskSpawnGroup=false, Mesh Sampling Type=::NewEnumerator5, Skeletal Mesh=$User.SK_Mesh, Socket Sampling Mode=::NewEnumerator5)
  ]
  particleUpdateScript [
   &ParticleState()
   &ScaleColor(
    Scale Alpha=&FloatFromCurve(FloatCurve=DataInterface() {propertyValues {Curve {keys [{arriveTangent=-0.1447, interpMode=RCIM_Cubic, time=0, value=0}, <at 0.4739 0.9561 RCIM_Cubic>, <at 1 0>]}}})
    Scale RGB=&VectorFromCurve(VectorCurve=DataInterface() {propertyValues {
       XCurve {keys [<at 0 0>,<at 0.5 5>,<at 1 0>]}
       YCurve {keys [<at 0 0>,<at 0.5 2.4374>,<at 1 0>]}
       ZCurve {keys [<at 0 0>,<at 0.5 0>,<at 1 0>]}
      }})
    ScaleA=true
    ScaleRGB=true
   )
   &CameraOffset(Camera Offset Amount=10)
   &DynamicMaterialParameters(
    Index 0 Param 1=&FloatFromCurve(FloatCurve=DataInterface() {propertyValues {Curve {keys [
         <at 0 -0.5 RCIM_Cubic>
         {arriveTangent=10, interpMode=RCIM_Cubic, leaveTangent=10, time=0.0692, value=0.2738}
         {arriveTangent=5.0777, interpMode=RCIM_Cubic, leaveTangent=5.0777, time=0.15, value=1}
         {interpMode=RCIM_Cubic, leaveTangent=4.2839, time=1, value=5}
        ]}}})
    Index 0 Param 2=1
    Index 0 Param 3=&FloatFromCurve(FloatCurve=DataInterface() {propertyValues {Curve {keys [{arriveTangent=0.1967, interpMode=RCIM_Cubic, time=0, value=0.8609}, {interpMode=RCIM_Cubic, leaveTangent=0.1844, time=1, value=1}]}}})
    Index 0 Param 4=1
    Module.Param0WriteEnabled=true
    Module.Param1WriteEnabled=true
    Module.Param2WriteEnabled=true
    Module.Param3WriteEnabled=true
   )
  ]
  rendererClasses [&NiagaraSpriteRendererProperties]
  renderers [{rendererClass=&NiagaraSpriteRendererProperties}]
 }
 Sparks {
  simTarget=GPUComputeSim
  emitterUpdateScript [
   &EmitterState(Life Cycle Mode=::NewEnumerator1, Loop Behavior=::NewEnumerator1, Loop Delay=0.1, Loop Duration=0.1, UseLoopDelay=true)
   &SpawnBurst_Instantaneous(Spawn Count=10, Use Spawn Probability=false, UseLoopCountLimit=false)
  ]
  particleSpawnScript [
   &InitializeParticle(
    Lifetime=&RandomRangeFloat(Minimum=0.1, Maximum=1)
    Position=&RandomRangeVector(Minimum=<-5 0 2>, Maximum=<15 0 2>)
    Position Mode=::NewEnumerator1
    Sprite Rotation Angle=&RandomRangeFloat(Minimum=0.02, Maximum=-0.02)
    Sprite Rotation Mode=::NewEnumerator1
    Sprite Size=&Vector2DFromFloat(Value=&RandomRangeFloat(Minimum=2.5, Maximum=5))
    Sprite Size Mode=::NewEnumerator4
    UsePositionOffset=false
    Write Material Random=true
   )
   &SkeletalMeshLocation(MaskSpawnGroup=false, Mesh Sampling Type=::NewEnumerator5, Skeletal Mesh=$User.SK_Mesh, Socket Sampling Mode=::NewEnumerator5)
   &Color(Color=&MakeLinearColorFromVectorAndFloat("Vector (RGB)"=<800 67.5149 0>, "Float (Alpha)"=&ClampFloat(Float=1)))
   &AddVelocity(Velocity=&Multiply_VectorByFloat(Vector=&UniformRangedVector(Minimum=<-400 -400 -400>, Maximum=<400 400 400>), Float=0.5)) {module=AddVelocity002}
   &SphereLocation(
    Hemisphere X=true
    MaskSpawnGroup=false
    Override Seed=false
    Sphere Radius=5
    Surface Only=false
    UseNonUniformScale=false
    UseOffset=false
    UseSphereOrientation=false
   ) {module=SphereLocation001}
   &ApplyInitialForces()
   &CameraOffset(Camera Offset Amount=40)
  ]
  particleUpdateScript [
   &ParticleState()
   &ScaleSpriteSize()
   &CurlNoiseForce(Noise Frequency=100, Noise Strength=100, Pan Noise=false)
   &AccelerationForce(Acceleration=&RandomRangeVector(Minimum=<0 0 -150>, Maximum=<0 0 -300>))
   &Drag(Drag=10, Use Linear Drag=true, Use Rotational Drag=true)
   &SolveForcesAndVelocity(Clamp Velocity=false, Limit Acceleration=false)
   &ScaleColor(
    Scale Alpha=&FloatFromCurve(FloatCurve=DataInterface() {propertyValues {Curve {keys [
         {arriveTangent=-0.7236, interpMode=RCIM_Cubic, time=0, value=0.0994}
         {arriveTangent=-0.0994, interpMode=RCIM_Cubic, leaveTangent=-0.0994, time=0.128, value=0.0373}
         <at 1 0 RCIM_Cubic>
        ]}}})
    Scale RGB=&VectorFromFloat(Value=&FloatFromCurve(FloatCurve=DataInterface() {propertyValues {Curve {keys [{arriveTangent=-14.4521, interpMode=RCIM_Cubic, time=0, value=15}, <at 0.4999 1.0559 RCIM_Cubic>]}}}))
    ScaleA=true
    ScaleRGB=true
   )
  ]
  rendererClasses [&NiagaraSpriteRendererProperties]
  renderers [{rendererClass=&NiagaraSpriteRendererProperties}]
 }
}
summary {
 systemName=P_GliderContrail_PirateOctopus_Deploy_Converted
 emitters [
  {emitterName=Coins, rendererClasses [&NiagaraMeshRendererProperties]}
  {emitterName=Glints, simTarget=GPUComputeSim, rendererClasses [&NiagaraSpriteRendererProperties]}
  {emitterName=Sparks, simTarget=GPUComputeSim, rendererClasses [&NiagaraSpriteRendererProperties]}
  {emitterName=FlashFlare, rendererClasses [&NiagaraSpriteRendererProperties]}
  {emitterName=Corona, rendererClasses [&NiagaraSpriteRendererProperties]}
  {emitterName=ShieldDecals, rendererClasses [&NiagaraSpriteRendererProperties]}
  {emitterName=GlowStart, rendererClasses [&NiagaraMeshRendererProperties]}
  {emitterName=ColdSmoke, rendererClasses [&NiagaraSpriteRendererProperties]}
  {emitterName=HQ_Light, rendererClasses [&NiagaraLightRendererProperties]}
 ]
 userVariables [User.SK_Mesh {
   type=NiagaraDataInterfaceSkeletalMesh
   defaultValue {
    struct=NiagaraExt_VariableValue_DataInterface
    value {dataInterface=&self:NiagaraDataInterfaceSkeletalMesh_0}
   }
  }]
}
