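# Niagara system for the PirateOctopus glider's idle contrail, apparently Cascade-converted, shedding gold coins amid a warm glow.
# Tumbling mesh coins fall from randomized looping bursts and movement-based SpawnPerUnit, with burst velocity and acceleration switching between gameplay and front end.
# A pulsing mesh aura, edge glows, curve-modulated GPU glints and orange-tinted InkySmokeRadial sprites use color curves, mostly with camera offsets.
# HQ_Light adds one persistent, distance-culled orange light.
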NiagaraSystem /BRCosmetics/Athena/Effects/Particles/Gliders/P_GliderContrail_PirateOctopus_Idle_Converted
dependencies {
 usedDataInterfaces [&NiagaraDataInterfaceCurve, &NiagaraDataInterfaceVectorCurve]
 usedDynamicInputs []
 usedModules [&SystemState]
}
emitters {
 Aura {
  emitterUpdateScript [
   &EmitterState(Life Cycle Mode=::NewEnumerator1, Loop Duration=1, UseLoopDelay=true)
   &SpawnRate(SpawnRate=5, Use Spawn Probability=false)
  ]
  particleSpawnScript [
   /SetVariables(Particles.MeshOrientation=&MakeQuatFromAxisAngle(Axis=&RandomRangeVector(Minimum=<0 0 0>, Maximum=<1 1 1>), AngleInDegrees=90)) {isSetParameters=true}
   &InitializeParticle(
    Lifetime=0.5
    Mesh Scale=/Niagara/DynamicInputs/CascadeConversion/SplitVector(
     Double Component Vector Channels:ECascadeNiagaraTwoVectorChannels=::NewEnumerator0
     Double Component:NiagaraFloat=&RandomRangeFloat(Minimum=0.7, Maximum=0.8)
     Single Component:NiagaraFloat=&RandomRangeFloat(Minimum=0.25, Maximum=0.456)
    )
    Mesh Scale Mode=::NewEnumerator4
    Position Offset=<0 0 -10>
    UsePositionOffset=true
    Write Material Random=true
   )
   &Color(Color=&MakeLinearColorFromVectorAndFloat("Vector (RGB)"=<25 11.3005 0>, "Float (Alpha)"=&ClampFloat(Float=10)))
   &DynamicMaterialParameters(
    Index 0 Param 2=&RandomRangeFloat(Minimum=0, Maximum=1)
    Index 0 Param 3=&RandomRangeFloat(Minimum=0, Maximum=1)
    Module.Param0WriteEnabled=false
    Module.Param1WriteEnabled=true
    Module.Param2WriteEnabled=true
    Module.Param3WriteEnabled=false
   ) {module=DynamicMaterialParameters001}
  ]
  particleUpdateScript [
   &ParticleState()
   &ScaleMeshSize(Scale Factor=&VectorFromCurve(VectorCurve=DataInterface() {propertyValues {XCurve {keys [<at 0 1>,<at 1 1.2>]}, YCurve {keys [<at 0 1>,<at 1 1.2>]}, ZCurve {keys [<at 0 1>,<at 1 1.2>]}}}))
   &ScaleColor(
    Scale Alpha=&FloatFromCurve(FloatCurve=DataInterface() {propertyValues {Curve {keys [
         <at 0 0 RCIM_Cubic>
         {arriveTangent=2.3525, interpMode=RCIM_Cubic, leaveTangent=2.3525, time=0.0821, value=0.0592}
         {arriveTangent=1.8815, interpMode=RCIM_Cubic, leaveTangent=1.8815, time=0.3487, value=0.8203}
         {arriveTangent=-0.1447, interpMode=RCIM_Cubic, leaveTangent=-0.1447, time=0.5851, value=1.0056}
         {arriveTangent=-1.5664, interpMode=RCIM_Cubic, leaveTangent=-1.5664, time=0.8941, value=0.7414}
         {arriveTangent=-7.0007, interpMode=RCIM_Cubic, leaveTangent=-7.0007, time=0.9922, value=0.368}
         <at 1 0 RCIM_Cubic>
        ]}}})
    Scale RGB=&VectorFromCurve(VectorCurve=DataInterface() {propertyValues {
       XCurve {keys [
         <at 0 0 RCIM_Cubic>
         {arriveTangent=6.4793, interpMode=RCIM_Cubic, leaveTangent=6.4793, time=0.3637, value=3.7556}
         {arriveTangent=-2.2466, interpMode=RCIM_Cubic, leaveTangent=-2.2466, time=0.5436, value=3.522}
         {arriveTangent=-11.4761, interpMode=RCIM_Cubic, leaveTangent=-11.4761, time=0.7, value=3}
         {arriveTangent=-10, interpMode=RCIM_Cubic, leaveTangent=-10, time=0.8402, value=0.1176}
         <at 1 0>
        ]}
       YCurve {keys [
         <at 0 0 RCIM_Cubic>
         {arriveTangent=2.4225, interpMode=RCIM_Cubic, leaveTangent=2.4225, time=0.3637, value=1.8887}
         {arriveTangent=-3.6356, interpMode=RCIM_Cubic, leaveTangent=-3.6356, time=0.5436, value=1.3168}
         {arriveTangent=-4.3869, interpMode=RCIM_Cubic, leaveTangent=-4.3869, time=0.7, value=0.6659}
         {arriveTangent=-2.2198, interpMode=RCIM_Cubic, leaveTangent=-2.2198, time=0.8402, value=0.0154}
         <at 1 0>
        ]}
       ZCurve {keys [
         <at 0 0 RCIM_Cubic>
         {arriveTangent=0.6526, interpMode=RCIM_Cubic, leaveTangent=0.6526, time=0.3637, value=0.2554}
         {arriveTangent=-0.4772, interpMode=RCIM_Cubic, leaveTangent=-0.4772, time=0.5436, value=0.3547}
         {arriveTangent=2.9098, interpMode=RCIM_Cubic, leaveTangent=2.9098, time=0.7, value=0.0949}
         {arriveTangent=-0.3164, interpMode=RCIM_Cubic, leaveTangent=-0.3164, time=0.8402, value=1.2179}
         <at 1 0>
        ]}
      }})
    ScaleA=true
    ScaleRGB=true
   )
   &DynamicMaterialParameters(
    Index 0 Param 1=&FloatFromCurve(FloatCurve=DataInterface() {propertyValues {Curve {keys [<at 0 1>,<at 0.25 1>,<at 1 0.5>]}}})
    Index 0 Param 4=&FloatFromCurve(FloatCurve=DataInterface() {propertyValues {Curve {keys [{arriveTangent=28.2423, interpMode=RCIM_Cubic, leaveTangent=28.2423, tangentMode=RCTM_User, time=0, value=1}, <at 1 5 RCIM_Cubic>]}}})
    Module.Param0WriteEnabled=true
    Module.Param1WriteEnabled=false
    Module.Param2WriteEnabled=false
    Module.Param3WriteEnabled=true
   )
   &CameraOffset(Camera Offset Amount=15)
  ]
  rendererClasses [&NiagaraMeshRendererProperties]
  renderers [{rendererClass=&NiagaraMeshRendererProperties, rendererIndex=0}]
 }
 Coin_Burst {
  emitterUpdateScript [
   &EmitterState(
    Life Cycle Mode=::NewEnumerator1
    Loop Delay=&RandomRangeFloat(Minimum=2, Maximum=5, Evaluation Type=::NewEnumerator1)
    Loop Duration=&RandomRangeFloat(Minimum=0.25, Maximum=3, Evaluation Type=::NewEnumerator1)
    Recalculate Duration Each Loop=true
    UseLoopDelay=true
   )
   &SpawnBurst_Instantaneous(Use Spawn Probability=false, UseLoopCountLimit=false)
   &SpawnBurst_Instantaneous(Spawn Time=0.25, Use Spawn Probability=false, UseLoopCountLimit=false) {module=SpawnBurst_Instantaneous001}
   &SpawnBurst_Instantaneous(
    Spawn Count=&RandomRangeInt(Minimum=0, Maximum=1)
    Spawn Time=0.133
    Use Spawn Probability=false
    UseLoopCountLimit=false
   ) {module=SpawnBurst_Instantaneous002}
  ]
  particleSpawnScript [
   /SetVariables(Particles.SpawnRotationVector=&RandomRangeVector(Minimum=<1 1 1>, Maximum=<3 3 3>)) {isSetParameters=true}
   /SetVariables(Particles.Acceleration:Vector3f=&FN_FrontEndVector3Modifier(
     Gameplay=&RandomRangeVector(Minimum=<0 0 -800>, Maximum=<0 50 -1200>)
     Front End=&RandomRangeVector(Minimum=<0 0 -800>, Maximum=<0 0 -1200>)
    )) {isSetParameters=true}
   /SetVariables(Particles.MeshOrientation=&MakeQuatFromAxisAngle(Axis=&RandomRangeVector(Minimum=<0 0 0>, Maximum=<1 1 1>), AngleInDegrees=90)) {isSetParameters=true}
   &InitializeParticle(
    Lifetime=&RandomRangeFloat(Minimum=1.25, Maximum=1.5)
    Mesh Scale=&SplitVector(
     Double Component Vector Channels:ECascadeNiagaraTwoVectorChannels=::NewEnumerator0
     Double Component:NiagaraFloat=&RandomRangeFloat(Minimum=0.04, Maximum=0.05)
     Single Component:NiagaraFloat=&RandomRangeFloat(Minimum=0.06, Maximum=0.08)
    )
    Mesh Scale Mode=::NewEnumerator4
    Position=&ConvertVectorToPosition(
     Input Position=&RandomRangeVector(Minimum=<0 -50 -25>, Maximum=<10 10 -30>)
     Conversion Operation:ENiagara_LWCConvertVecToPos=::NewEnumerator0
    )
    Position Mode=::NewEnumerator1
    UsePositionOffset=false
    Write Material Random=true
   )
   &AddVelocity(
    Velocity=&FN_FrontEndVector3Modifier(Gameplay=<0 0 0>, Front End=<0 -200 -500>)
    Velocity Speed Scale=0.5
   )
   &InheritVelocity(Inherited Velocity Amount Scale=<0 0 0.1>, Inherited Velocity Speed Limit=0, Speed Limit=false, UseSpeedThreshold=false)
   &Color(Color=&MakeLinearColorFromVectorAndFloat("Vector (RGB)"=<1 0.8784 0.4941>, "Float (Alpha)"=&ClampFloat(Float=1)))
   &ApplyInitialForces()
   &ShapeLocation(MaskSpawnGroup=false, Override Seed=false, Sphere Radius=5, UseOwnerScale=true)
  ]
  particleUpdateScript [
   &ParticleState()
   &ScaleMeshSize(Scale Factor=<1.2 1.2 1.2>)
   &UpdateMeshOrientation(Rotation Vector=$Particles.SpawnRotationVector)
   &AccelerationForce(Acceleration=$Particles.Acceleration)
   &SolveForcesAndVelocity(Clamp Velocity=false, Limit Acceleration=false)
  ]
  rendererClasses [&NiagaraMeshRendererProperties]
  renderers [{rendererClass=&NiagaraMeshRendererProperties, rendererIndex=0}]
 }
 Coin_Distance {
  emitterUpdateScript [
   &EmitterState(
    Life Cycle Mode=::NewEnumerator1
    Loop Delay=0.03
    Loop Duration=&RandomRangeFloat(Minimum=1, Maximum=2, Evaluation Type=::NewEnumerator1)
    Recalculate Duration Each Loop=true
    UseLoopDelay=true
   )
   &SpawnPerUnit(
    Module.Use Max Movement Threshold=true
    Module.Use Movement Tolerance=true
    Movement Tolerance=0.1
    Spawn Spacing:NiagaraFloat=&Divide_Float(A=1, B=&Divide_Float(A=1, B=1000))
    Use Spawn Probability=false
   )
  ]
  particleSpawnScript [
   /SetVariables(Particles.SpawnRotationVector=&RandomRangeVector(Minimum=<1 1 1>, Maximum=<3 3 3>)) {isSetParameters=true}
   /SetVariables(Particles.Acceleration:Vector3f=&RandomRangeVector(Minimum=<0 0 -800>, Maximum=<0 0 -1200>)) {isSetParameters=true}
   /SetVariables(Particles.MeshOrientation=&MakeQuatFromAxisAngle(Axis=&RandomRangeVector(Minimum=<0 0 0>, Maximum=<1 1 1>), AngleInDegrees=90)) {isSetParameters=true}
   &InitializeParticle(
    Lifetime=&RandomRangeFloat(Minimum=1.25, Maximum=1.5)
    Mesh Scale=&SplitVector(
     Double Component Vector Channels:ECascadeNiagaraTwoVectorChannels=::NewEnumerator0
     Double Component:NiagaraFloat=&RandomRangeFloat(Minimum=0.04, Maximum=0.05)
     Single Component:NiagaraFloat=&RandomRangeFloat(Minimum=0.06, Maximum=0.08)
    )
    Mesh Scale Mode=::NewEnumerator4
    Position=&ConvertVectorToPosition(
     Input Position=&RandomRangeVector(Minimum=<-10 -20 0>, Maximum=<10 20 0>)
     Conversion Operation:ENiagara_LWCConvertVecToPos=::NewEnumerator0
    )
    Position Mode=::NewEnumerator1
    UsePositionOffset=false
    Write Material Random=true
   )
   &InheritVelocity(Inherited Velocity Amount Scale=<0 0 0.1>, Inherited Velocity Speed Limit=0, Speed Limit=false, UseSpeedThreshold=false)
   &Color(Color=&MakeLinearColorFromVectorAndFloat("Vector (RGB)"=<1 0.8784 0.4941>, "Float (Alpha)"=&ClampFloat(Float=1)))
   &ApplyInitialForces()
   &ShapeLocation(MaskSpawnGroup=false, Override Seed=false, Sphere Radius=5, UseOwnerScale=true)
  ]
  particleUpdateScript [
   &ParticleState()
   &ScaleMeshSize(Scale Factor=<1.2 1.2 1.2>)
   &UpdateMeshOrientation(Rotation Vector=$Particles.SpawnRotationVector)
   &AccelerationForce(Acceleration=$Particles.Acceleration)
   &SolveForcesAndVelocity(Clamp Velocity=false, Limit Acceleration=false)
  ]
  rendererClasses [&NiagaraMeshRendererProperties]
  renderers [{rendererClass=&NiagaraMeshRendererProperties, rendererIndex=0}]
 }
 EdgeGlows {
  emitterUpdateScript [
   &EmitterState(Life Cycle Mode=::NewEnumerator1, Loop Duration=1, UseLoopDelay=true)
   &SpawnRate(SpawnRate=30, Use Spawn Probability=false)
  ]
  particleSpawnScript [
   &InitializeParticle(
    Lifetime=&RandomRangeFloat(Minimum=0.3, Maximum=0.4)
    Position=&ConvertVectorToPosition(
     Input Position=&RandomRangeVector(Minimum=<-5 -20 -10>, Maximum=<0 20 -10>)
     Conversion Operation:ENiagara_LWCConvertVecToPos=::NewEnumerator0
    )
    Position Mode=::NewEnumerator1
    Sprite Size=&Vector2DFromFloat(Value=&RandomRangeFloat(Minimum=30, Maximum=40))
    Sprite Size Mode=::NewEnumerator4
    UsePositionOffset=false
    Write Material Random=true
   )
   &Color(Color=&MakeLinearColorFromVectorAndFloat("Vector (RGB)"=<8 4.6658 0.0943>, "Float (Alpha)"=&ClampFloat(Float=0.1)))
   &AddVelocity(Velocity=&RandomRangeVector(Minimum=<0 0 -10>, Maximum=<0 0 10>))
   &DynamicMaterialParameters(
    Index 0 Param 1=&RandomRangeFloat(Minimum=2.5, Maximum=3)
    Module.Param0WriteEnabled=true
    Module.Param1WriteEnabled=false
    Module.Param2WriteEnabled=false
    Module.Param3WriteEnabled=false
   )
  ]
  particleUpdateScript [
   &ParticleState()
   &ScaleSpriteSize()
   &AccelerationForce(Acceleration=<0 0 2>)
   &SolveForcesAndVelocity(Clamp Velocity=false, Limit Acceleration=false)
   &DynamicMaterialParameters(
    Index 0 Param 2=1
    Index 0 Param 3=1
    Index 0 Param 4=1
    Module.Param0WriteEnabled=false
    Module.Param1WriteEnabled=true
    Module.Param2WriteEnabled=true
    Module.Param3WriteEnabled=true
   ) {module=DynamicMaterialParameters001}
   &CameraOffset(Camera Offset Amount=15)
   &ScaleColor(
    Scale Alpha=&FloatFromCurve(FloatCurve=DataInterface() {propertyValues {Curve {keys [<at 0 0 RCIM_Cubic>,<at 0.4 1 RCIM_Cubic>,<at 1 0>]}}}, Scale Curve=1.5)
    Scale RGB=&VectorFromCurve(VectorCurve=DataInterface() {propertyValues {
       XCurve {keys [<at 0 1 RCIM_Cubic>,<at 1 0.7747>]}
       YCurve {keys [<at 0 0.5177 RCIM_Cubic>,<at 1 0.8>]}
       ZCurve {keys [<at 0 0.0387 RCIM_Cubic>,<at 1 0>]}
      }})
    ScaleA=true
    ScaleRGB=true
   )
  ]
  rendererClasses [&NiagaraSpriteRendererProperties]
  renderers [{rendererClass=&NiagaraSpriteRendererProperties}]
 }
 Glints {
  simTarget=GPUComputeSim
  emitterUpdateScript [
   &EmitterState(
    Life Cycle Mode=::NewEnumerator1
    Loop Duration=&RandomRangeFloat(Minimum=2, Maximum=3, Evaluation Type=::NewEnumerator1)
    Recalculate Duration Each Loop=true
    UseLoopDelay=true
   )
   &SpawnRate(
    SpawnRate=&Multiply_Float(
     A=5
     B=&FloatFromCurve(
      FloatCurve=DataInterface() {propertyValues {Curve {keys [
          <at -0.0016 -0.0061 RCIM_Cubic>
          {arriveTangent=11.061, interpMode=RCIM_Cubic, leaveTangent=11.061, time=0.0346, value=0.0848}
          {arriveTangent=0.793, interpMode=RCIM_Cubic, leaveTangent=0.793, time=0.0597, value=0.672}
          {arriveTangent=1.7145, interpMode=RCIM_Cubic, leaveTangent=1.7145, time=0.088, value=0.1271}
          {arriveTangent=0.632, interpMode=RCIM_Cubic, leaveTangent=0.632, time=0.1981, value=0.9092}
          {arriveTangent=0.0147, interpMode=RCIM_Cubic, leaveTangent=0.0147, time=0.3854, value=0.3151}
          {arriveTangent=-0.5126, interpMode=RCIM_Cubic, leaveTangent=-0.5126, time=0.5929, value=0.915}
          <at 1 0 RCIM_Cubic>
         ]}}}
      CurveIndex=$Emitter.NormalizedLoopAge
     )
    )
    Use Spawn Probability=false
   )
  ]
  particleSpawnScript [
   &InitializeParticle(
    Lifetime=&RandomRangeFloat(Minimum=0.25, Maximum=0.8)
    Position=&ConvertVectorToPosition(
     Input Position=&RandomRangeVector(Minimum=<-15 -10 0>, Maximum=<15 10 0>)
     Conversion Operation:ENiagara_LWCConvertVecToPos=::NewEnumerator0
    )
    Position Mode=::NewEnumerator1
    Sprite Size=&RandomRangeVector2D(Minimum=<25 5>, Maximum=<100 150>)
    Sprite Size Mode=::NewEnumerator4
    UsePositionOffset=false
    Write Material Random=true
   )
   &Color(Color=&MakeLinearColorFromVectorAndFloat("Vector (RGB)"=<1 1 1>, "Float (Alpha)"=&ClampFloat(Float=1)))
   &ShapeLocation(
    Hemisphere Distribution=<0.5 1>
    MaskSpawnGroup=false
    Override Seed=false
    Rotation Mode=::NewEnumerator2
    Sphere Radius=5
    UseOwnerScale=true
    Yaw / Pitch / Roll=<0.75 0 0>
   )
   &CameraOffset(Camera Offset Amount=10)
  ]
  particleUpdateScript [
   &ParticleState()
   &ScaleSpriteSize()
   &ScaleColor(
    Scale RGB=&VectorFromCurve(VectorCurve=DataInterface() {propertyValues {XCurve {keys [<at 0 1>,<at 0.5 1>]}, YCurve {keys [<at 0 0.4452>,<at 0.5 0.7033>]}, ZCurve {keys [<at 0 0.1082>,<at 0.5 0>]}}})
    ScaleA=true
    ScaleRGB=true
   )
  ]
  rendererClasses [&NiagaraSpriteRendererProperties]
  renderers [{rendererClass=&NiagaraSpriteRendererProperties}]
 }
 HQ_Light {
  emitterUpdateScript [
   &EmitterState(
    Enable Distance Culling=true
    Inactive Response=::NewEnumerator1
    Life Cycle Mode=::NewEnumerator1
    Loop Behavior=::NewEnumerator1
    Loop Duration Mode=::NewEnumerator1
    MaxDistance=1500
    Module.bUseMaxDistance=true
    Module.bUseMinDistance=false
    Scalability Mode=::NewEnumerator1
    UseLoopDelay=false
   )
   &SpawnBurst_Instantaneous(Spawn Count=1, Spawn Probability=HlslExpression() {hlslExpression="Emitter.LoopCount == 0"}, Use Spawn Probability=true)
  ]
  particleSpawnScript [&InitializeParticle(Color Mode=::NewEnumerator0, UsePositionOffset=false, Write Material Random=true)]
  particleUpdateScript [
   &ParticleState(
    Kill Particles When Lifetime Has Elapsed=false
    Let Infinitely Lived Particles Die When Emitter Deactivates=true
    Loop Particles Lifetime=true
   )
   &FN_Light_Attributes(
    Color=<rgba 15 6.5 0 0.7>
    Create Color Attribute=true
    Create Exponent Attribute=true
    Create Position Attribute=true
    Create Radius Attribute=true
    Enable Visibility Tag=true
    Exponent=7.5
    Position=<-5 -2.5 -20>
    Radius=50
    Scale Color=2
    VisibilityTag=HlslExpression() {hlslExpression="Particles.UniqueID && 1"}
   )
  ]
  rendererClasses [&NiagaraLightRendererProperties]
  renderers [{rendererClass=&NiagaraLightRendererProperties, rendererIndex=0}]
 }
 InkySmokeRadial {
  emitterUpdateScript [
   &EmitterState(Life Cycle Mode=::NewEnumerator1, Loop Duration=1, UseLoopDelay=true)
   &SpawnRate(SpawnRate=10, Use Spawn Probability=false)
  ]
  particleSpawnScript [
   /SetVariables(Particles.SpriteRotationRate=&RandomRangeFloat(Minimum=-0.02, Maximum=0.02)) {isSetParameters=true}
   &InitializeParticle(
    Lifetime=&RandomRangeFloat(Minimum=0.5, Maximum=1)
    Position=&ConvertVectorToPosition(
     Input Position=&RandomRangeVector(Minimum=<-2 -10 -5>, Maximum=<2 10 5>)
     Conversion Operation:ENiagara_LWCConvertVecToPos=::NewEnumerator0
    )
    Position Mode=::NewEnumerator1
    Sprite Rotation Mode=::NewEnumerator2
    Sprite Size=&Vector2DFromFloat(Value=&RandomRangeFloat(Minimum=30, Maximum=60))
    Sprite Size Mode=::NewEnumerator4
    UsePositionOffset=false
    Write Material Random=true
   )
   &Color(Color=&MakeLinearColorFromVectorAndFloat("Vector (RGB)"=<250 250 250>, "Float (Alpha)"=0.7))
   &DynamicMaterialParameters(
    Index 0 Param 2=&RandomRangeFloat(Minimum=0, Maximum=1)
    Module.Param0WriteEnabled=false
    Module.Param1WriteEnabled=true
    Module.Param2WriteEnabled=false
    Module.Param3WriteEnabled=false
   ) {module=DynamicMaterialParameters001}
  ]
  particleUpdateScript [
   &ParticleState()
   &SpriteRotationRate(Rotation Rate=&AngleConversion(
     Angle:NiagaraFloat=$Particles.SpriteRotationRate
     Angle Input:ENiagara_AngleInput=::NewEnumerator1
     Angle Output:ENiagara_AngleInput=::NewEnumerator0
    ))
   /SetVariables(Particles.SubImageIndex=&RandomRangeFloat(Minimum=0, Maximum=4)) {isSetParameters=true}
   &ScaleColor(
    Scale Alpha=&FloatFromCurve(FloatCurve=DataInterface() {propertyValues {Curve {keys [
         <at 0 0 RCIM_Cubic>
         {arriveTangent=0.1173, interpMode=RCIM_Cubic, leaveTangent=0.1173, tangentMode=RCTM_User, time=0.224, value=1}
         {arriveTangent=-1.2887, interpMode=RCIM_Cubic, leaveTangent=-1.2887, time=0.6, value=1}
         {interpMode=RCIM_Cubic, leaveTangent=-1.9896, time=1, value=0}
        ]}}})
    Scale RGB=&VectorFromCurve(VectorCurve=DataInterface() {propertyValues {
       XCurve {keys [
         <at 0.05 1 RCIM_Cubic>
         {arriveTangent=-0.1823, interpMode=RCIM_Cubic, leaveTangent=-0.1823, time=0.16, value=0.4}
         {arriveTangent=0.2679, interpMode=RCIM_Cubic, leaveTangent=0.2679, time=0.25, value=0.9635}
         <at 1 0.625>
        ]}
       YCurve {keys [
         <at 0.05 0.2274 RCIM_Cubic>
         {arriveTangent=1.0378, interpMode=RCIM_Cubic, leaveTangent=1.0378, time=0.16, value=0.1966}
         {arriveTangent=0.0877, interpMode=RCIM_Cubic, leaveTangent=0.0877, time=0.25, value=0.4349}
         <at 1 0.2703>
        ]}
       ZCurve {keys [
         <at 0.05 0 RCIM_Cubic>
         <at 0.16 0 RCIM_Cubic>
         {arriveTangent=0.016, interpMode=RCIM_Cubic, leaveTangent=0.016, time=0.25, value=0}
         <at 1 0.0134>
        ]}
      }})
    ScaleA=true
    ScaleRGB=true
   )
   &DynamicMaterialParameters(
    Index 0 Param 1=&FloatFromCurve(FloatCurve=DataInterface() {propertyValues {Curve {keys [
         <at 0.0703 2.299 RCIM_Cubic>
         {arriveTangent=-3.8695, interpMode=RCIM_Cubic, leaveTangent=-3.8695, tangentMode=RCTM_User, time=1, value=0.25}
        ]}}})
    Index 0 Param 2=&RandomRangeFloat(Minimum=0, Maximum=1)
    Index 0 Param 3=&FloatFromCurve(FloatCurve=DataInterface() {propertyValues {Curve {keys [
         {arriveTangent=0.5688, interpMode=RCIM_Cubic, time=-0.0045, value=2.5544}
         {arriveTangent=-1.5474, interpMode=RCIM_Cubic, leaveTangent=-1.5474, time=0.1073, value=1.6357}
         {interpMode=RCIM_Cubic, leaveTangent=0.7013, time=1, value=1}
        ]}}})
    Index 0 Param 4=1
    Module.Param0WriteEnabled=true
    Module.Param1WriteEnabled=false
    Module.Param2WriteEnabled=true
    Module.Param3WriteEnabled=true
   )
   &CameraOffset(Camera Offset Amount=20)
  ]
  rendererClasses [&NiagaraSpriteRendererProperties]
  renderers [{rendererClass=&NiagaraSpriteRendererProperties}]
 }
}
summary {
 systemName=P_GliderContrail_PirateOctopus_Idle_Converted
 emitters [
  {emitterName=Coin_Distance, rendererClasses [&NiagaraMeshRendererProperties]}
  {emitterName=Coin_Burst, rendererClasses [&NiagaraMeshRendererProperties]}
  {emitterName=Glints, simTarget=GPUComputeSim, rendererClasses [&NiagaraSpriteRendererProperties]}
  {emitterName=InkySmokeRadial, rendererClasses [&NiagaraSpriteRendererProperties]}
  {emitterName=EdgeGlows, rendererClasses [&NiagaraSpriteRendererProperties]}
  {emitterName=Aura, rendererClasses [&NiagaraMeshRendererProperties]}
  {emitterName=HQ_Light, rendererClasses [&NiagaraLightRendererProperties]}
 ]
 userVariables [
  User.bIsFrontEnd {type=NiagaraBool}
  User.bIsFrontEndPreview {type=NiagaraBool}
  User.SK_Mesh {
   type=NiagaraDataInterfaceSkeletalMesh
   defaultValue {
    struct=NiagaraExt_VariableValue_DataInterface
    value {dataInterface=&self:NiagaraDataInterfaceSkeletalMesh_0}
   }
  }
 ]
}
